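; BP_HiddenChestActor is a hidden chest secret object that plays audio when its discovered state changes.
; It overrides UpdateMaterialVisibility, calling the parent implementation and then PlayRevealAudio with the discovered flag.
; PlayRevealAudio runs only where cosmetic events can execute, playing the EnterAudio component when discovered and ExitAudio otherwise.
; BeginPlay, overlap and Tick events are unused stubs.
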(event EventBeginPlay
  (|CallParentFunction) ; This node is disabled and will not be called.\nDrag off pins to build functionality.
)

(event Collision|EventActorBeginOverlap
  (params (OtherActor :type Actor%20Object%20Reference))
  (|CallParentFunction OtherActor) ; This node is disabled and will not be called.\nDrag off pins to build functionality.
)

(event EventTick
  (params (DeltaSeconds :type Float%20%28single-precision%29))
  (|CallParentFunction DeltaSeconds) ; This node is disabled and will not be called.\nDrag off pins to build functionality.
)

(event Custom|PlayRevealAudio
  (params (bDiscovered :type Boolean))
  (Networking|CanExecuteCosmeticEvents
    (:True
      (if bDiscovered
        (Audio|Components|Audio|Play (Variables|Default|GetEnterAudio) 0.0)
        (else
          (Audio|Components|Audio|Play (Variables|Default|GetExitAudio) 0.0))))))

(fn UpdateMaterialVisibility
  (params (bDiscovered :type Boolean))
  (|CallParentFunction bDiscovered)
  (CallFunction|PlayRevealAudio bDiscovered))

(fn ConstructionScript
  (|CallParentFunction))
