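; Hidden moving key item that travels along a referenced spline with looping audio, dropping an item pickup when shot or damaged below a health threshold.
; The server drives a timeline, reversing it on LOS trigger overlaps and pausing it while players overlap the discovery component.
; Clients follow a replicated transform with optional tick interpolation.
; Breaking stops movement, tosses loot and plays client effects; a footstep-trail routine spawns decals via ground traces.
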(event Custom|DebugTest
  ; For debugging only!
  (Rendering|Debug|DrawDebugSphere (Transformation|GetActorLocation) 1e+02 12 (Math|Color|SelectColor "(R=0.054902,G=1.000000,B=0.000000,A=1.000000)" "(R=1.000000,G=0.000000,B=0.009056,A=1.000000)" (Networking|HasAuthority)) 0.1)
  (Networking|SwitchHasAuthority
    (:Remote
      (Development|PrintString (Utilities|String|Append "TickEnabled=" (Utilities|String|ToString(Boolean) (Actor|Tick|IsActorTickEnabled))) true false "(R=0.000000,G=0.660000,B=1.000000,A=1.000000)" 0.0))))

(event EventBeginPlay
  ; Init/ Spawn Item and Move
  (|CallParentFunction)
  (SoundIndicator|StartSoundIndicatorLooping (Variables|Default|GetFortSoundIndicator) (Transformation|GetActorLocation self) "/Hera/Audio/POI/Temple/SC_Temple_InvisibleKey_Amb.SC_Temple_InvisibleKey_Amb" 1.0 0)
  (Networking|SwitchHasAuthority
    (:Authority
      _)
    (:Remote
      (Variables|/Script/Engine.PrimitiveComponent|SetbGenerateOverlapEvents false (Variables|Default|GetLOS%20Trigger)) ; Disable LOS overlaps on client for perf
    ))
  (Variables|Replication|SetNetUpdateFrequency (Variables|Default|GetMovementNetUpdateFrequency))
  (if (Variables|Default|GetbUseInterpolation)
    (Networking|SwitchHasAuthority
      (:Authority
        _)
      (:Remote
        (CallFunction|SetTemporarilyTicking true)))
    (else
      _))
  (Networking|SwitchHasAuthority
    (:Authority
      (Variables|BuildingActor|SetbIsInvulnerable false)
      (Variables|Actor|SetbCanBeDamaged true)
      (Health|SetHealth self (Variables|Default|GetHealth%20))
      ;; Move Key Item on Spline
      (CallFunction|MoveObject))
    (:Remote)))

(event Custom|Footstep%20Udate
  (bind _lastfootsteplocation (Variables|Default|GetLastFootstepLocation))
  (bind _returnvalue (Transformation|GetActorLocation))
  (bind _returnvalue_2 (- _returnvalue _lastfootsteplocation))
  (bind _returnvalue_3 (Math|Vector|Normalize _returnvalue_2))
  (bind _returnvalue_4 (Math|Vector|VectorLength _returnvalue_3))
  ; Footsteps Decal/FX set up
  (Utilities|FlowControl|Sequence
    (:then_0
      (if (> _returnvalue_4 0)
        (bind _output_get (Variables|Default|SetFootsteps%20Update%20Time%20Elapsed (+ (Variables|Default|GetFootsteps%20Update%20Time) (Variables|Default|GetFootsteps%20Update%20Time%20Elapsed))))
        ; Line Check/Spawn Effect
        (if (>= _output_get (/ _returnvalue_4 (Variables|Default|GetSteps%20Per%20Velocity)))
          (bind (_outhit _returnvalue_5) (Collision|LineTraceForObjects (Transformation|GetWorldLocation (Variables|Default|GetArrow)) (+ (Transformation|GetWorldLocation (Variables|Default|GetArrow)) (* "0.000000,0.000000,-1.000000" 100)) (Utilities|Array|MakeArray "ObjectTypeQuery1" "ObjectTypeQuery2")))
          (if _returnvalue_5
            (CallFunction|SpawnEffect (.ImpactPoint (Collision|BreakHitResult _outhit)))
            (Variables|Default|SetFootsteps%20Update%20Time%20Elapsed 0.0)))
        (else
          (Variables|Default|SetFootsteps%20Update%20Time%20Elapsed 0.0))))
    (:then_1
      (Variables|Default|SetLastFootstepLocation (Transformation|GetActorLocation)))))

(event Custom|Move%20Object
  ; Timeline/Spline Ref
  (Networking|SwitchHasAuthority
    (:Authority
      (Components|Timeline|SetPlayRate (Variables|BP_HiddenMovingKeyItem|GetMover_Timeline) (/ 1.0 (Variables|Default|GetTime%20To%20Complete)))
      (CallFunction|PlayTimeline))))

(event Custom|PlayTimeline
  ; Timeline/Spline Ref
  (if (not (Variables|Default|GetPot%20Is%20Damaged))
    (bind (_direction _move_track) (Timeline|Mover_Timeline Play)
      (:Update
        (Transformation|SetActorLocationAndRotation (Spline|GetLocationatDistanceAlongSpline (Variables|/Game/Athena/Artemis/Environments/Blueprints/Foundations/Artemis_BoatSpline.Artemis_BoatSpline_C|GetBoatSpline (Variables|Default|GetSpline%20Ref)) (Math|Float|Lerp 0.0 (Spline|GetSplineLength (Variables|/Game/Athena/Artemis/Environments/Blueprints/Foundations/Artemis_BoatSpline.Artemis_BoatSpline_C|GetBoatSpline (Variables|Default|GetSpline%20Ref))) _move_track) "World") (Math|Rotator|CombineRotators (Spline|GetRotationatDistanceAlongSpline (Variables|/Game/Athena/Artemis/Environments/Blueprints/Foundations/Artemis_BoatSpline.Artemis_BoatSpline_C|GetBoatSpline (Variables|Default|GetSpline%20Ref)) (Math|Float|Lerp 0.0 (Spline|GetSplineLength (Variables|/Game/Athena/Artemis/Environments/Blueprints/Foundations/Artemis_BoatSpline.Artemis_BoatSpline_C|GetBoatSpline (Variables|Default|GetSpline%20Ref))) _move_track) "World") (Math|Rotator|SelectRotator "0,180,0" "0, 0, 0" (Variables|Default|GetbReversed))))
        (Variables|Default|SetRepActorTransform (Transformation|GetActorTransform))))))

(event OnComponentBeginOverlap(LOSTrigger)
  (params (OverlappedComponent :type Primitive%20Component%20Object%20Reference) (OtherActor :type Actor%20Object%20Reference) (OtherComp :type Primitive%20Component%20Object%20Reference) (OtherBodyIndex :type Integer) (bFromSweep :type Boolean) (SweepResult :type Hit%20Result%20Structure%20%28by%20ref%29))
  ; Reverse Trigger (Run away!)
  (Networking|SwitchHasAuthority
    (:Authority
      (Utilities|Casting|CastToFortPlayerPawnAthena OtherActor)
      (if (not (Variables|Default|GetPot%20Is%20Damaged))
        (Variables|Default|SetbReversed (not (Variables|Default|GetbReversed)))
        (Transformation|AddActorWorldRotation "0, 180, 0")
        (Variables|Default|SetRepActorTransform (Transformation|GetActorTransform))
        (if (Variables|Default|GetbReversed)
          (CallFunction|ReverseTimeline)
          (else
            (CallFunction|PlayTimeline)))))))

(event Custom|StopTimeline
  ; Timeline/Spline Ref
  (Timeline|Mover_Timeline Stop))

(event Custom|ReverseTimeline
  ; Timeline/Spline Ref
  (if (not (Variables|Default|GetPot%20Is%20Damaged))
    (Timeline|Mover_Timeline Reverse)))

(event OnComponentBeginOverlap(DiscoveryComponent)
  (params (OverlappedComponent :type Primitive%20Component%20Object%20Reference) (OtherActor :type Actor%20Object%20Reference) (OtherComp :type Primitive%20Component%20Object%20Reference) (OtherBodyIndex :type Integer) (bFromSweep :type Boolean) (SweepResult :type Hit%20Result%20Structure%20%28by%20ref%29))
  ; Trigger Events
  (Networking|SwitchHasAuthority
    (:Authority
      (bind _asfort_player_pawn_athena (Utilities|Casting|CastToFortPlayerPawnAthena OtherActor))
      (Utilities|FlowControl|Sequence
        (:then_0
          (CallFunction|StopTimeline))
        (:then_1
          (Environment|ApplyGlobalEnvironmentGameplayEffecttoActor _asfort_player_pawn_athena "/Game/Athena/Environments/Blueprints/DudeBro/GE_Athena_FallDmgImmune_FX.GE_Athena_FallDmgImmune_FX_C"))))))

(event EventOnDamageServer
  (params (Damage :type Float%20%28single-precision%29) (DamageTags :type Gameplay%20Tag%20Container%20Structure%20%28by%20ref%29) (Momentum :type Vector :default "0, 0, 0") (HitInfo :type Hit%20Result%20Structure%20%28by%20ref%29) (InstigatedBy :type Controller%20Object%20Reference) (DamageCauser :type Actor%20Object%20Reference) (EffectContext :type Gameplay%20Effect%20Context%20Handle%20Structure))
  (bind (_bblockinghit _binitialoverlap _time _distance _location) (Collision|BreakHitResult HitInfo))
  ; Damage Logic
  (if (< (Health|GetHealthPercent self) 0.25)
    (Utilities|FlowControl|DoOnce false
      (:Completed
        (Utilities|FlowControl|Sequence
          (:then_0
            (Networking|SwitchHasAuthority
              (:Authority
                (Variables|/Script/SecretObjectRuntime.FortSecretObjectComponent|SetbUndiscoverWhenOutOfRange false (Variables|DiscoveryWall|GetDiscoveryComponent)))))
          (:then_1
            (bind _output_get (Variables|Default|SetPot%20Is%20Damaged true))
            (Collision|SetCollisionEnabled (Variables|Debug|GetStaticMeshComponent) "NoCollision")
            (CallFunction|PlayDamagedAudioandFX _location)
            (Utilities|Time|PauseTimerbyHandle (Variables|Default|GetTimer%20Handle))
            (Utilities|FlowControl|Sequence
              (:then_0
                (CallFunction|StopTimeline))
              (:then_1
                ; Spawn Loot
                (if _output_get
                  (Utilities|FlowControl|DoOnce false
                    (:Completed
                      (Networking|SwitchHasAuthority
                        (:Authority
                          (Audio|Components|Audio|Stop (Variables|Default|GetAudio))
                          (bind _returnvalue (Inventory|SpawnPickupInWorldWithLevel (Variables|Default|GetItem%20Definition) 0 1 (Transformation|GetWorldLocation (Variables|Default|GetArrow)) (* (Transformation|GetForwardVector (Variables|Default|GetArrow)) (Math|Conversions|ToVector(Float) (Variables|Default|GetToss%20Velocity)))))
                          (bind _asfort_pickup_athena (Utilities|Casting|CastToFortPickupAthena _returnvalue))
                          (Pickup|TossPickupinDirection _asfort_pickup_athena (+ (Math|Vector|MakeVector (* (Variables|Default|GetToss%20Spread%20Multiplier) (Math|Random|RandomFloatinRange -1e+02 1e+02)) (* (Math|Random|RandomFloatinRange -1e+02 1e+02) (Variables|Default|GetToss%20Spread%20Multiplier)) (* (Math|Random|RandomFloatinRange 5e+01 1.5e+02) (Variables|Default|GetToss%20Spread%20Multiplier))) (* (Transformation|GetForwardVector (Variables|Default|GetArrow)) (Math|Conversions|ToVector(Float) (Variables|Default|GetToss%20Velocity)))) 0 false false)
                          (Development|PrintString "Spawn Loot" true true "(R=0.000000,G=0.660000,B=1.000000,A=1.000000)" 1e+01))
                        (:Remote
                          (Audio|PlaySoundatLocation "/Game/Sounds/Foley_Loot/Containers/Pirate_Booty/Pirate_Booty_Beacon_Activate_Cue.Pirate_Booty_Beacon_Activate_Cue" (Transformation|GetWorldLocation (Variables|Debug|GetStaticMeshComponent)))
                          (Niagara|SpawnSystematLocation "/Game/Effects/Fort_Effects/Gameplay/PirateChests/NS_Loot_Chest.NS_Loot_Chest" (Transformation|GetWorldLocation (Variables|Default|GetBillboard))))))))))))))))

(event OnComponentEndOverlap(DiscoveryComponent)
  (params (OverlappedComponent :type Primitive%20Component%20Object%20Reference) (OtherActor :type Actor%20Object%20Reference) (OtherComp :type Primitive%20Component%20Object%20Reference) (OtherBodyIndex :type Integer))
  ; Leave Trigger Events
  (Networking|SwitchHasAuthority
    (:Authority
      (Utilities|Casting|CastToFortPlayerPawnAthena OtherActor)
      (if (== (Utilities|Array|Length (Collision|GetOverlappingActors (Variables|DiscoveryWall|GetDiscoveryComponent) "/Script/FortniteGame.FortPlayerPawnAthena" :!class /Script/Engine.PrimitiveComponent)) 0)
        (if (Variables|Default|GetPot%20Is%20Damaged)
          _
          (elif (Variables|Default|GetbReversed)
            (CallFunction|ReverseTimeline)
            (else
              (CallFunction|PlayTimeline))))))))

(event Custom|PlayDamagedAudioAndFX
  (params (HitVector :type Vector))
  ; Client Cosmetics
  (Networking|SwitchHasAuthority
    (:Remote
      (Audio|PlaySoundatLocation "/Hera/Audio/POI/Temple/SC_Temple_InvisibleKey_Pickup.SC_Temple_InvisibleKey_Pickup" HitVector)
      (Niagara|SpawnSystematLocation "/Game/Effects/Fort_Effects/Gameplay/Player/Tactical/NS_Player_HealthShield_Damage_Effect_TacElim.NS_Player_HealthShield_Damage_Effect_TacElim" HitVector))))

(event EventTick
  (params (DeltaSeconds :type Float%20%28single-precision%29))
  ; Client Tick (interpolate towards replicated transform)
  (if (Variables|Default|GetbTransformInit)
    (Transformation|SetActorTransform (Math|Interpolation|TInterpTo (Transformation|GetActorTransform) (Variables|Default|GetRepActorTransform) DeltaSeconds (Variables|Default|GetMovementInterpSpeed)))))

(event EventEndPlay
  (params (EndPlayReason :type EEndPlayReason%20Enum :default "Destroyed"))
  ; End Play
  (Utilities|Time|ClearandInvalidateTimerbyHandle (Variables|Default|GetFootsteps%20Timer%20Handle)))

(fn OnRep_PotIsDamaged
  (bind _staticmeshcomponent (Variables|Debug|GetStaticMeshComponent))
  (bind _fx_orb (Variables|Default|GetFX%20Orb))
  (if (Variables|Default|GetPot%20Is%20Damaged)
    (Rendering|SetVisibility _staticmeshcomponent false false)
    (Collision|SetCollisionEnabled _staticmeshcomponent "NoCollision")
    (Niagara|SpawnSystematLocation "/Game/Athena/Items/QuestParents/PropQuestActor/NS_QuestItem_Breakable.NS_QuestItem_Breakable" (Transformation|GetWorldLocation (Variables|Debug|GetStaticMeshComponent)))
    (Niagara|SetNiagaraVariable(Bool) _fx_orb "FadeRibbon" true)
    (Components|Activation|Deactivate _fx_orb)
    (Rendering|SetVisibility (Variables|Default|GetPointLight) false false)))

(fn OnRep_RepActorTransform
  (if (Variables|Default|GetbUseInterpolation)
    (if (Variables|Default|GetbTransformInit)
      _
      (else
        _))
    (else
      _))
  (Transformation|SetActorTransform (Variables|Default|GetRepActorTransform))
  (Variables|Default|SetbTransformInit true))

(fn SpawnEffect
  (params (Location :type Vector))
  (bind _returnvalue (Game|SpawnActorfromClass (Variables|Default|GetFootstepDecalClass) (Math|Transform|MakeTransform Location (Math|Rotator|CombineRotators (Transformation|GetActorRotation) (Math|Rotator|MakeRotator 0.0 0.0 9e+01))) "Undefined" "MultiplyWithRoot" self 0))
  (Actor|SetLifeSpan _returnvalue (Variables|Default|GetFootstepEffectLifespan))
  (Effects|SpawnFXSystemAttached :FXSystemAsset "/Hera/SecretObject/VFX/NS_HiddenMovingKey_CrumbTrail.NS_HiddenMovingKey_CrumbTrail" :AttachToComponent (Actor|GetComponentbyClass _returnvalue "/Script/Engine.SceneComponent") :AttachPointName "None" :Rotation "90.000000,0.000000,0.000000" :Scale "1.000000,1.000000,1.000000" :LocationType "SnapToTarget" :bAutoDestroy true :PoolingMethod "None" :bAutoActivate true))

(fn ConstructionScript)
