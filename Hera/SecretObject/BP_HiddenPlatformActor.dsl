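; Hidden temple platform actor that adds cosmetic energy FX and audio when its discovered state changes.
; BeginPlay creates dynamic materials for the platform and corner FX meshes and starts them at zero opacity.
; The UpdateVisibility override calls the parent, then plays or reverses a timeline driving GlobalOpacity and plays enter or exit audio.
; All FX and audio logic runs only where cosmetic events can execute.
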(event EventBeginPlay
  (|CallParentFunction)
  (Networking|CanExecuteCosmeticEvents
    (:True
      ; Set FX Mesh Visibility
      (Rendering|SetVisibility (Variables|Default|GetSM_FX_TemplePlatform_Magic) true false)
      (Rendering|SetVisibility (Variables|Default|GetSM_FX_Temple_PlatformCornerFX) true false)
      ; Set materials for FX components
      (bind _returnvalue (Rendering|Material|CreateDynamicMaterialInstance (Variables|Default|GetSM_FX_TemplePlatform_Magic) 0 "/Hera/FXEnvironment/Temple/Materials/MI_Env_Temple_PlatformEnergy.MI_Env_Temple_PlatformEnergy" "None" :!class /Script/Engine.PrimitiveComponent))
      (bind _output_get (Variables|Default|SetDynamic%20FX%20Platform%20Material _returnvalue))
      (Rendering|Material|SetScalarParameterValue _output_get "GlobalOpacity" 0.0 :!class /Script/Engine.MaterialInstanceDynamic)
      (bind _returnvalue_2 (Rendering|Material|CreateDynamicMaterialInstance (Variables|Default|GetSM_FX_Temple_PlatformCornerFX) 0 "/Hera/FXEnvironment/Temple/Materials/MI_Env_Temple_PlatformCornerMesh.MI_Env_Temple_PlatformCornerMesh" "None" :!class /Script/Engine.PrimitiveComponent))
      (bind _output_get_2 (Variables|Default|SetDynamic%20FX%20Corner%20Material _returnvalue_2))
      (Rendering|Material|SetScalarParameterValue _output_get_2 "GlobalOpacity" 0.0 :!class /Script/Engine.MaterialInstanceDynamic))))

(event Custom|PlayRevealTimelineAnim
  (params (bDiscovered :type Boolean))
  (Networking|CanExecuteCosmeticEvents
    (:True
      (if bDiscovered
        (bind (_direction _fade_track_inverse_fx) (Timeline|Timeline Play)
          (:Update
            (Rendering|Material|SetScalarParameterValue (Variables|Default|GetDynamic%20FX%20Platform%20Material) "GlobalOpacity" _fade_track_inverse_fx :!class /Script/Engine.MaterialInstanceDynamic)
            (Rendering|Material|SetScalarParameterValue (Variables|Default|GetDynamic%20FX%20Corner%20Material) "GlobalOpacity" _fade_track_inverse_fx :!class /Script/Engine.MaterialInstanceDynamic)))
        (else
          (Timeline|Timeline Reverse))))))

(event Custom|PlayRevealAudio
  (params (bDiscovered :type Boolean))
  (Networking|CanExecuteCosmeticEvents
    (:True
      (if bDiscovered
        (Audio|Components|Audio|Play (Variables|Default|GetEnter%20Audio) 0.0)
        (else
          (Audio|Components|Audio|Play (Variables|Default|GetExit%20Audio) 0.0)
          (Audio|Components|Audio|Stop (Variables|Default|GetEnter%20Audio)))))))

(fn UpdateVisibility
  (params (bDiscovered :type Boolean))
  (|CallParentFunction bDiscovered)
  (CallFunction|PlayRevealTimelineAnim bDiscovered)
  (CallFunction|PlayRevealAudio bDiscovered))

(fn ConstructionScript)
