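; Hidden temple floor secret object whose fountain shuts off and whose vault opens once enough switches activate.
; Activate Temple Vault counts switches, then fades the fountain, removes floor collision, lowers water, broadcasts an accolade verb and cleans up after delays.
; Separately, the authority shuts off the fountain when any support actor dies.
; RepNotify flags drive fountain FX, lights and floor state, with remote BeginPlay emptying an already-destroyed fountain.
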(event EventBeginPlay
  ; Init
  (|CallParentFunction)
  (Networking|SwitchHasAuthority
    (:Authority
      (for _array_element (Variables|Default|GetFountainSupportActors)
        (Utilities|IsValid _array_element
          (:Is%20Valid
            (EventDispatchers|AssignOnDied _array_element (delegate Custom|OnFountainSupportDied))))))
    (:Remote
      (if (Variables|Default|GetFountainDestroyed)
        (CallFunction|EmptyFountain)))))

(event Custom|OnFountainSupportDied
  (params (DamagedActor :type Actor%20Object%20Reference) (Damage :type Float%20%28single-precision%29) (InstigatedBy :type Controller%20Object%20Reference) (DamageCauser :type Actor%20Object%20Reference) (HitLocation :type Vector) (FHitComponent :type Primitive%20Component%20Object%20Reference) (BoneName :type Name) (Momentum :type Vector))
  ; Init
  (Networking|SwitchHasAuthority
    (:Authority
      (CallFunction|UnbindFountainDestroyEvents) ; One support getting destroyed shuts off fountain,\ndon't need to hear about others
      (Variables|Default|SetFountainDestroyed true))))

(event Custom|UnbindFountainDestroyEvents
  ; Init
  (for _array_element (Variables|Default|GetFountainSupportActors)
    (Utilities|IsValid _array_element
      (:Is%20Valid
        (EventDispatchers|UnbindEventfromOnDied _array_element (delegate Custom|OnFountainSupportDied))))))

(event EventEndPlay
  (params (EndPlayReason :type EEndPlayReason%20Enum :default "Destroyed"))
  ; End Play
  (|CallParentFunction EndPlayReason)
  (Networking|SwitchHasAuthority
    (:Authority
      (CallFunction|UnbindFountainDestroyEvents))))

(event Custom|Activate%20Temple%20Vault
  (params (Pawn%20Ref :type Fort%20Pawn%20Object%20Reference))
  ; Open Temple Vault Event
  (Math|Integer|IncrementInt (Variables|Default|GetSwitches%20Active)
    (:%20%20
      (if (>= (Variables|Default|GetSwitches%20Active) (Variables|Default|GetSwitches%20Required))
        (bind _output_get (Variables|Default|SetPawn%20Ref Pawn%20Ref))
        (GameplayCue|ExecuteGameplayCue _output_get (Variables|Default|GetGameplay%20Cue%20Tag) (Utilities|Struct|MakeGameplayEffectContextHandle))
        ;; Fade Floor first then activate vault
        (Utilities|FlowControl|Sequence
          (:then_0
            (CallFunction|TurnOffFountain))
          (:then_1
            (Utilities|FlowControl|Delay (Variables|Default|GetVault%20Open%20Delay))
            (Variables|Default|SetTemple%20Vault%20is%20Active true)
            ;; Remove Water
            (Class|BPTempleWaterPlaneEvent|MoveWaterPlane (Variables|Default|GetWater%20Plane%20Ref))
            (Development|PrintString "Vault is Open!" true true "(R=0.000000,G=0.660000,B=1.000000,A=1.000000)" 1e+01)
            ;;[0.15 0.15 0.15 0.5 15] Custom Verb Message to trigger the Accolade
            (CustomVerb|BroadcastCustomVerbMessage (CustomVerb|MakeCustomVerbMessage (Variables|Default|GetPawn%20Ref) "(TagName=\"CustomVerb.S42.Temple.VaultOpen\")"))
            ;; Clean up FX/Lighting
            (Utilities|FlowControl|Delay (Variables|Default|GetClean%20Up%20Delay))
            (CallFunction|CleanUpTempleEvent)
            (Development|PrintString "Clean Up Temple" true true "(R=0.000000,G=0.660000,B=1.000000,A=1.000000)" 1e+01)))))))

(event Custom|Clean%20Up%20Temple%20Event
  ; Deactivate Temple Event
  (Networking|SwitchHasAuthority
    (:Authority
      (Variables|Default|SetTemple%20Vault%20is%20Active false)
      (Components|Activation|Deactivate (Variables|Default|GetFX_Fountain))
      (Rendering|SetVisibility (Variables|Default|GetSM_FountainWaterfalls) false false)
      (Components|Activation|Deactivate (Variables|Default|GetMagic%20FX)))))

(event Custom|TurnOffFountain
  ; Vault Opening FX | Turn Off Fountain FX
  (Utilities|FlowControl|Sequence
    (:then_0
      (Networking|SwitchHasAuthority
        (:Authority
          (Variables|/Script/SecretObjectRuntime.FortSecretObjectComponent|SetbUndiscoverWhenOutOfRange false (Variables|DiscoveryWall|GetDiscoveryComponent)) ; Disable proximity vanishing behavior
        )))
    (:then_1
      (Components|Activation|Activate (Variables|Default|GetMagic%20FX) false)
      (Niagara|SetNiagaraVariable(Bool) (Variables|Default|GetFX_Fountain) "KillFountain" true)
      (bind _returnvalue (Rendering|Material|CreateDynamicMaterialInstance (Variables|Default|GetSM_FountainWaterfalls) 0 "/Hera/FXEnvironment/Temple/Materials/MI_Env_Temple_Fountain_Water_Spline.MI_Env_Temple_Fountain_Water_Spline" "None" :!class /Script/Engine.PrimitiveComponent))
      (Utilities|FlowControl|Sequence
        (:then_0
          (Audio|AddEvent (GameInstanceSubsystems|GetMusicEventSubsystem) "/Hera/Audio/POI/Temple/MSS_MX_S42_POI_Temple_PuzzleSolve_OneShotSingle.MSS_MX_S42_POI_Temple_PuzzleSolve_OneShotSingle" "(TagName=\"MusicEventSubsystem.Event.POI.Gameplay\")" "(GameplayTags=)"))
        (:then_1
          (bind (_direction _waterfall_opacity _pointlight_intensity) (Timeline|Waterfall_Timeline Play)
            (:Update
              (Rendering|Material|SetScalarParameterValue _returnvalue "FinalOpacity" _waterfall_opacity :!class /Script/Engine.MaterialInstanceDynamic)
              (Rendering|Components|Light|SetIntensity (Variables|Default|GetPointLight1) _pointlight_intensity))
            (:Finished
              (Networking|SwitchHasAuthority
                (:Authority
                  (Variables|Default|SetFountainIsEmpty true))))))))))

(event Custom|DisableFountain
  ; Vault Opening FX | Turn Off Fountain FX
  (Niagara|SetNiagaraVariable(Bool) (Variables|Default|GetFX_Fountain) "KillFountain" true)
  (bind _returnvalue (Rendering|Material|CreateDynamicMaterialInstance (Variables|Default|GetSM_FountainWaterfalls) 0 "/Hera/FXEnvironment/Temple/Materials/MI_Env_Temple_Fountain_Water_Spline.MI_Env_Temple_Fountain_Water_Spline" "None" :!class /Script/Engine.PrimitiveComponent))
  (Utilities|FlowControl|Sequence
    (:then_0
      (Audio|AddEvent (GameInstanceSubsystems|GetMusicEventSubsystem) "/Hera/Audio/POI/Temple/MSS_MX_S42_POI_Temple_PuzzleSolve_OneShotSingle.MSS_MX_S42_POI_Temple_PuzzleSolve_OneShotSingle" "(TagName=\"MusicEventSubsystem.Event.POI.Gameplay\")" "(GameplayTags=)"))
    (:then_1
      (Timeline|Waterfall_Timeline Play))))

(event Custom|EmptyFountain
  ; Vault Opening FX | Turn Off Fountain FX
  (Networking|CanExecuteCosmeticEvents
    (:True
      (Components|Activation|Deactivate (Variables|Default|GetFX_Fountain))
      (bind _returnvalue (Rendering|Material|CreateDynamicMaterialInstance (Variables|Default|GetSM_FountainWaterfalls) 0 "/Hera/FXEnvironment/Temple/Materials/MI_Env_Temple_Fountain_Water_Spline.MI_Env_Temple_Fountain_Water_Spline" "None" :!class /Script/Engine.PrimitiveComponent))
      (Rendering|Material|SetScalarParameterValue _returnvalue "FinalOpacity" 0.0 :!class /Script/Engine.MaterialInstanceDynamic)
      (Rendering|Components|Light|SetIntensity (Variables|Default|GetPointLight1) 0.0))))

(fn OnRep_FountainDestroyed
  (if (Variables|Default|GetFountainDestroyed)
    (CallFunction|DisableFountain)))

(fn OnRep_FountainIsEmpty
  (if (Variables|Default|GetFountainIsEmpty)
    (CallFunction|EmptyFountain)))

(fn OnRep_IsInsideTrigger)

(fn OnRep_TempleVaultisActive
  (bind _staticmeshcomponent (Variables|Debug|GetStaticMeshComponent))
  (bind _pointlight1 (Variables|Default|GetPointLight1))
  (bind _pointlight2 (Variables|Default|GetPointLight2))
  (bind _pointlight3 (Variables|Default|GetPointLight3))
  (bind _pointlight (Variables|Default|GetPointLight))
  (if (Variables|Default|GetTemple%20Vault%20is%20Active)
    ; Activate/Deactive Point Lights
    (Rendering|SetVisibility _pointlight true false)
    (Rendering|SetVisibility _pointlight1 true false)
    (Rendering|SetVisibility _pointlight2 true false)
    (Rendering|SetVisibility _pointlight3 true false)
    (Development|PrintString "Turn on FX")
    (Rendering|SetVisibility _staticmeshcomponent false false)
    (Collision|SetCollisionEnabled _staticmeshcomponent "NoCollision")
    (Development|PrintString "Remove Door/Collision")
    (else
      (Rendering|SetVisibility _pointlight false false)
      (Rendering|SetVisibility _pointlight1 false false)
      (Rendering|SetVisibility _pointlight2 false false)
      (Rendering|SetVisibility _pointlight3 false false))))

(fn ConstructionScript)
