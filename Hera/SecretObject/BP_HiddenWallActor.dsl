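; BP_HiddenWallActor is a secret-object hidden wall that adds sound to its parent class's visibility change.
; UpdateVisibility calls the parent with bDiscovered, then calls PlayRevealAudio with the same flag.
; PlayRevealAudio plays only where cosmetic events can execute.
; Its branch is fixed to true rather than reading bDiscovered, so it always plays ActivateAudio and never reaches DeactivateAudio.
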(event EventBeginPlay)

(event Collision|EventActorBeginOverlap
  (params (OtherActor :type Actor%20Object%20Reference)))

(event EventTick
  (params (DeltaSeconds :type Float%20%28single-precision%29)))

(event Custom|PlayRevealAudio
  (params (bDiscovered :type Boolean))
  (Networking|CanExecuteCosmeticEvents
    (:True
      (if true
        (Audio|Components|Audio|Play (Variables|Default|GetActivateAudio) 0.0)
        (else
          (Audio|Components|Audio|Play (Variables|Default|GetDeactivateAudio) 0.0))))))

(fn UpdateVisibility
  (params (bDiscovered :type Boolean))
  (|CallParentFunction bDiscovered)
  (CallFunction|PlayRevealAudio bDiscovered))

(fn ConstructionScript
  (|CallParentFunction))
