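; Secret wall actor Blueprint that adds reveal and hide sounds to a parent class's discoverable wall.
; The UpdateVisibility override calls the parent, then passes the discovered flag to PlayRevealAduio (spelled that way).
; That event plays the activate or deactivate audio component to match the flag, only where cosmetic events can execute.
; BeginPlay, overlap and Tick events add no logic.
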(event EventBeginPlay
  (|CallParentFunction) ; This node is disabled and will not be called.\nDrag off pins to build functionality.
)

(event Collision|EventActorBeginOverlap
  (params (OtherActor :type Actor%20Object%20Reference)))

(event EventTick
  (params (DeltaSeconds :type Float%20%28single-precision%29)))

(event Custom|PlayRevealAduio
  (params (bDiscovered :type Boolean))
  (Networking|CanExecuteCosmeticEvents
    (:True
      (if bDiscovered
        (Audio|Components|Audio|Play (Variables|Default|GetActivateAudio) 0.0)
        (else
          (Audio|Components|Audio|Play (Variables|Default|GetDeactivateAudio) 0.0))))))

(fn UpdateVisibility
  (params (bDiscovered :type Boolean))
  (|CallParentFunction bDiscovered)
  (CallFunction|PlayRevealAduio bDiscovered))

(fn ConstructionScript
  (|CallParentFunction))
