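# Material function that builds UVs for four material layers and outputs a WSM scale scalar for each.
# Each layer switches statically between scaled UVs with fixed or time-panned offsets and a FlipBook trim cell, chosen by index or animated with optional phase offset.
# Layers 2-4 choose their source UV set through static bools; only layer 1 supports independent horizontal/vertical tiling.
# The WSM Modulator multiplies feed no output.
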MaterialFunction /Hera/SecretObject/Materials/MF_UVsSet_CustomUVs
graph {
 properties {description=This function checks through UV Coordinate's 0 through 3}
 nodes [
  n1 = ScalarParameter() {defaultValue=512, group=1_Layer UV, parameterName=WSM Scale 1, sortPriority=1}
  FunctionOutput(None=@n1) {outputName=WSM 1, sortPriority=9}
  n2 = ScalarParameter() {defaultValue=512, group=4_layer UV, parameterName=WSM Scale 4, sortPriority=1}
  FunctionOutput(None=@n2) {outputName=WSM 4, sortPriority=12}
  n3 = ScalarParameter() {defaultValue=1, group=1_Layer UV, parameterName=Trim Column Count, sortPriority=1}
  n4 = ScalarParameter() {defaultValue=1, group=1_Layer UV, parameterName=Trim Rows Count, sliderMin=1, sortPriority=3}
  n5 = Multiply(A=@n3, B=@n4)
  n6 = Divide(A=ScalarParameter() {group=2_Layer UV, parameterName="L1 Trim UV Anim Offset ##", sortPriority=5}, B=@n5)
  n7 = Frac(None=Multiply(A=Time(), B=ScalarParameter() {defaultValue=1, group=1_Layer UV, parameterName=L1 Trim UV Animation Speed, sortPriority=4}))
  n8 = StaticSwitchParameter(True=Frac(None=Add(A=@n6, B=@n7)), False=@n7) {group=2_Layer UV, parameterName=L1 Trim UV Anim Offset, sortPriority=5}
  n9 = ScalarParameter() {defaultValue=1, group=1_Layer UV, parameterName="Trim ##", sortPriority=1}
  n10 = Divide(
   A=StaticSwitchParameter(True=Multiply(A=@n8, B=@n5), False=Add(B=@n9) {constA=-1}) {group=1_Layer UV, parameterName=L1 Trim UV Animation, sortPriority=4}
   B=Multiply(A=@n4, B=@n3)
  )
  n11 = NamedRerouteUsage()
  n12 = ScalarParameter() {defaultValue=1, group=1_Layer UV, parameterName=UV Scale Horizontal}
  n13 = ScalarParameter() {defaultValue=1, group=1_Layer UV, parameterName=UV Scale Vertical}
  n14 = StaticSwitchParameter(
   True=Multiply(A=@n11, B=AppendVector(A=@n12, B=@n13))
   False=Multiply(A=@n11, B=ScalarParameter() {defaultValue=1, group=1_Layer UV, parameterName=UV Tiling, sortPriority=1})
  ) {group=1_Layer UV, parameterName=IndependentUVTiling}
  n15 = AppendVector(
   A=ScalarParameter() {group=1_Layer UV, parameterName="Offset  X", sortPriority=3}
   B=ScalarParameter() {group=1_Layer UV, parameterName="Offset  Y", sortPriority=3}
  )
  n16 = StaticSwitchParameter(True=Multiply(A=@n15, B=Time()), False=@n15) {group=1_Layer UV, parameterName=L1 Panner, sortPriority=-1}
  FunctionOutput(None=StaticSwitchParameter(
    True=MaterialFunctionCall("Animation  Phase (0-1)"=@n10, Number of Rows=@n4, Number of Columns=@n3, UVs=@n11) {materialFunction=&FlipBook}.UVs
    False=Add(A=@n14, B=@n16)
   ) {group=1_Layer UV, parameterName=L1 Trim UV, sortPriority=-1}) {outputName=UVs Layer 1}
  n17 = ScalarParameter() {defaultValue=1, group=2_Layer UV, parameterName=L2 Trim Column Count, sortPriority=1}
  n18 = ScalarParameter() {defaultValue=1, group=2_Layer UV, parameterName=L2 Trim Row Count, sliderMin=1, sortPriority=3}
  n19 = Multiply(A=@n17, B=@n18)
  n20 = Divide(A=ScalarParameter() {group=2_Layer UV, parameterName="L2 Trim UV Anim Offset ##", sortPriority=5}, B=@n19)
  n21 = Frac(None=Multiply(A=Time(), B=ScalarParameter() {defaultValue=1, group=2_Layer UV, parameterName=L2 Trim UV Animation Speed, sortPriority=4}))
  n22 = StaticSwitchParameter(True=Frac(None=Add(A=@n20, B=@n21)), False=@n21) {group=2_Layer UV, parameterName=L2 Trim UV Anim Offset, sortPriority=5}
  n23 = ScalarParameter() {defaultValue=1, group=2_Layer UV, parameterName="L2 Trim ##", sortPriority=1}
  n24 = Divide(
   A=StaticSwitchParameter(True=Multiply(A=@n22, B=@n19), False=Add(B=@n23) {constA=-1}) {group=2_Layer UV, parameterName=L2 Trim UV Animation, sortPriority=4}
   B=Multiply(A=@n18, B=@n17)
  )
  n25 = StaticSwitch(
   True=NamedRerouteUsage()
   False=NamedRerouteUsage()
   Value=FunctionInput(Preview=StaticBool()) {bUsePreviewValueAsDefault=true, inputName=L2 UV2, inputType=FunctionInput_StaticBool}
  )
  n26 = Multiply(A=@n25, B=ScalarParameter() {defaultValue=1, group=2_Layer UV, parameterName=UV Scale 2, sortPriority=1})
  n27 = AppendVector(
   A=ScalarParameter() {group=2_Layer UV, parameterName=Offset X 2, sortPriority=3}
   B=ScalarParameter() {group=2_Layer UV, parameterName=Offset Y 2, sortPriority=3}
  )
  n28 = StaticSwitchParameter(True=Multiply(A=@n27, B=Time()), False=@n27) {group=2_Layer UV, parameterName=L2 Panner, sortPriority=-1}
  FunctionOutput(None=StaticSwitchParameter(
    True=MaterialFunctionCall("Animation  Phase (0-1)"=@n24, Number of Rows=@n18, Number of Columns=@n17, UVs=@n25) {materialFunction=&FlipBook}.UVs
    False=Add(A=@n26, B=@n28)
   ) {group=2_Layer UV, parameterName=L2 Trim UV, sortPriority=-1}) {outputName=UVs Layer 2, sortPriority=1}
  n29 = ScalarParameter() {defaultValue=1, group=3_Layer UV, parameterName=L3 Trim Column Count, sortPriority=1}
  n30 = ScalarParameter() {defaultValue=1, group=3_Layer UV, parameterName=L3 Trim Rows Count, sliderMin=1, sortPriority=3}
  n31 = Multiply(A=@n29, B=@n30)
  n32 = Divide(A=ScalarParameter() {group=2_Layer UV, parameterName="L3 Trim UV Anim Offset ##", sortPriority=5}, B=@n31)
  n33 = Frac(None=Multiply(A=Time(), B=ScalarParameter() {defaultValue=1, group=3_Layer UV, parameterName=L3 Trim UV Animation Speed, sortPriority=4}))
  n34 = StaticSwitchParameter(True=Frac(None=Add(A=@n32, B=@n33)), False=@n33) {group=2_Layer UV, parameterName=L3 Trim UV Anim Offset, sortPriority=5}
  n35 = ScalarParameter() {defaultValue=1, group=3_Layer UV, parameterName="L3 Trim ##", sortPriority=1}
  n36 = Divide(
   A=StaticSwitchParameter(True=Multiply(A=@n34, B=@n31), False=Add(B=@n35) {constA=-1}) {group=3_Layer UV, parameterName=L3 Trim UV Animation, sortPriority=4}
   B=Multiply(A=@n30, B=@n29)
  )
  n37 = StaticSwitch(
   True=NamedRerouteUsage()
   False=NamedRerouteUsage()
   Value=FunctionInput(Preview=StaticBool()) {bUsePreviewValueAsDefault=true, inputName=L3 UV2, inputType=FunctionInput_StaticBool}
  )
  n38 = Multiply(A=@n37, B=ScalarParameter() {defaultValue=1, group=3_Layer UV, parameterName=UV Scale 3, sortPriority=1})
  n39 = AppendVector(
   A=ScalarParameter() {group=3_Layer UV, parameterName=Offset X 3, sortPriority=3}
   B=ScalarParameter() {group=3_Layer UV, parameterName=Offset Y 3, sortPriority=3}
  )
  n40 = StaticSwitchParameter(True=Multiply(A=@n39, B=Time()), False=@n39) {group=3_Layer UV, parameterName=L3 Panner, sortPriority=-1}
  FunctionOutput(None=StaticSwitchParameter(
    True=MaterialFunctionCall("Animation  Phase (0-1)"=@n36, Number of Rows=@n30, Number of Columns=@n29, UVs=@n37) {materialFunction=&FlipBook}.UVs
    False=Add(A=@n38, B=@n40)
   ) {group=3_Layer UV, parameterName=L3 Trim UV, sortPriority=-1}) {outputName=UVs Layer 3, sortPriority=2}
  n41 = ScalarParameter() {defaultValue=1, group=4_layer UV, parameterName=L4 Trim Column Count, sortPriority=1}
  n42 = ScalarParameter() {defaultValue=1, group=4_layer UV, parameterName=L4 Trim Rows Count, sliderMin=1, sortPriority=3}
  n43 = Multiply(A=@n41, B=@n42)
  n44 = Divide(A=ScalarParameter() {group=2_Layer UV, parameterName="L4 Trim UV Anim Offset ##", sortPriority=5}, B=@n43)
  n45 = Frac(None=Multiply(A=Time(), B=ScalarParameter() {defaultValue=1, group=4_layer UV, parameterName=L4 Trim UV Animation Speed, sortPriority=4}))
  n46 = StaticSwitchParameter(True=Frac(None=Add(A=@n44, B=@n45)), False=@n45) {group=2_Layer UV, parameterName=L4 Trim UV Anim Offset, sortPriority=5}
  n47 = ScalarParameter() {defaultValue=1, group=4_layer UV, parameterName="L4 Trim ##", sortPriority=1}
  n48 = Divide(
   A=StaticSwitchParameter(True=Multiply(A=@n46, B=@n43), False=Add(B=@n47) {constA=-1}) {group=4_layer UV, parameterName=L4 Trim UV Animation, sortPriority=4}
   B=Multiply(A=@n42, B=@n41)
  )
  n49 = StaticSwitch(
   True=NamedRerouteUsage()
   False=NamedRerouteUsage()
   Value=FunctionInput(Preview=StaticBool()) {bUsePreviewValueAsDefault=true, inputName=L4 UV2, inputType=FunctionInput_StaticBool}
  )
  n50 = Multiply(A=@n49, B=ScalarParameter() {defaultValue=1, group=4_layer UV, parameterName=UV Scale 4, sortPriority=1})
  n51 = AppendVector(
   A=ScalarParameter() {group=4_layer UV, parameterName=Offset X 4, sortPriority=3}
   B=ScalarParameter() {group=4_layer UV, parameterName=Offset Y 4, sortPriority=3}
  )
  n52 = StaticSwitchParameter(True=Multiply(A=@n51, B=Time()), False=@n51) {group=4_Layer UV, parameterName=L4 Panner, sortPriority=-1}
  FunctionOutput(None=StaticSwitchParameter(
    True=MaterialFunctionCall("Animation  Phase (0-1)"=@n48, Number of Rows=@n42, Number of Columns=@n41, UVs=@n49) {materialFunction=&FlipBook}.UVs
    False=Add(A=@n50, B=@n52)
   ) {group=4_layer UV, parameterName=L4 Trim UV, sortPriority=-1}) {outputName=UVs Layer 4, sortPriority=3}
  n53 = ScalarParameter() {defaultValue=512, group=3_Layer UV, parameterName=WSM Scale 3, sortPriority=1}
  FunctionOutput(None=@n53) {outputName=WSM 3, sortPriority=11}
  n54 = ScalarParameter() {defaultValue=512, group=2_Layer UV, parameterName=WSM Scale 2, sortPriority=1}
  FunctionOutput(None=@n54) {outputName=WSM 2, sortPriority=9}
  Multiply(A=@n54, B=VectorParameter() {group=2_layer UV, parameterName=WSM Modulator 2, sortPriority=2, defaultValue=<rgba 1 1 1 0>}.RGB)
  Multiply(A=@n1, B=VectorParameter() {group=1_layer UV, parameterName=WSM Modulator, sortPriority=2, defaultValue=<rgba 1 1 1 0>}.RGB)
  Multiply(A=MaterialFunctionCall(Number=@n24) {materialFunction=&DebugScalarValues}.Result)
  Multiply(A=@n2, B=VectorParameter() {group=4_layer UV, parameterName=WSM Modulator 4, sortPriority=2, defaultValue=<rgba 1 1 1 0>}.RGB)
  Multiply(A=@n53, B=VectorParameter() {group=3_layer UV, parameterName=WSM Modulator 3, sortPriority=2, defaultValue=<rgba 1 1 1 0>}.RGB)
  n55 = FunctionInput(Preview=TextureCoordinate()) {
   bUsePreviewValueAsDefault=true
   desc=UV1 needs to remain consistent with mesh tangents to work properly.
   inputName=UV1
   inputType=FunctionInput_Vector2
   sortPriority=50
  }
  NamedRerouteDeclaration(None=@n55) {name=UV1, nodeColor=<rgba 1 0 0 1>}
  NamedRerouteDeclaration(None=FunctionInput(Preview=TextureCoordinate() {coordinateIndex=1}) {bUsePreviewValueAsDefault=true, inputName=UV2, inputType=FunctionInput_Vector2, sortPriority=51}) {name=UV2, nodeColor=<rgba 1 0 0 1>}
  TextureCoordinate()
  TextureCoordinate()
  TextureCoordinate()
  TextureCoordinate()
  TextureCoordinate() {coordinateIndex=1}
  TextureCoordinate() {coordinateIndex=1}
  TextureCoordinate() {coordinateIndex=1}
 ]
}
