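# Material function for SecretObject that outputs a scalar hit mask around HitPosition, fading over HitDuration from HitTime.
# Normalized, saturated elapsed time drives a one-minus-square-root factor that divides pixel distance, shrinking the spot quickly at first.
# The result is normalized by HitRadius, inverted, then boosted and clamped for a solid core with a short falloff.
# PixelPosition defaults to world position without shader offsets.
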MaterialFunction /Hera/SecretObject/Materials/MF_SecretObject_Hit_Original
graph {nodes [
  n1 = WorldPosition() {worldPositionShaderOffset=WPT_ExcludeAllShaderOffsets}
  n2 = Time()
  n3 = FunctionInput() {inputName=HitTime, inputType=FunctionInput_Scalar, sortPriority=10}
  n4 = FunctionInput() {bUsePreviewValueAsDefault=true, inputName=HitDuration, inputType=FunctionInput_Scalar, sortPriority=10, previewValue=<1 0 0 1>}
  n5 = Saturate(None=Divide(A=Subtract(A=@n2, B=@n3), B=@n4))
  n6 = Divide(
   A=Distance(
    A=FunctionInput(Preview=@n1.XYZ) {bUsePreviewValueAsDefault=true, inputName=PixelPosition, sortPriority=20}
    B=FunctionInput() {inputName=HitPosition}
   )
   B=OneMinus(None=Power(Base=@n5) {constExponent=0.5})
  )
  n7 = FunctionInput() {bUsePreviewValueAsDefault=true, inputName=HitRadius, inputType=FunctionInput_Scalar, sortPriority=10, previewValue=<100 0 0 1>}
  n8 = OneMinus(None=Saturate(None=Divide(A=@n6, B=@n7)))
  FunctionOutput(None=Saturate(None=Multiply(A=@n8) {constB=3}))
 ]}
