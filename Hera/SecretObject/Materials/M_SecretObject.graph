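# Material M_SecretObject for secret objects, building animated, procedural cell-and-noise patterns driven by time and world, actor and object positions.
# Three custom PCG hash noises sample scale-compensated grids, staggered by permuted irrational-like constants (1.618, 1.3247, pi), then are summed.
# A saturated mask blends two material-attribute sets, keeping the first's vertex attributes.
# Composite subgraphs cover shared WPO and fire effects, seasonal moss/RVT, tangent-to-world conversion and mobile optimization.
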Material /Hera/SecretObject/Materials/M_SecretObject
graph {
 expressions {
  MaterialExpressionAbs_0 {class=MaterialExpressionAbs, inputs [None {from=MaterialExpressionScalarParameter_3, output=""}], properties {}}
  MaterialExpressionAbs_1 {class=MaterialExpressionAbs, inputs [None {from=MaterialExpressionVertexNormalWS_4, output=""}], properties {}}
  MaterialExpressionAbs_2 {class=MaterialExpressionAbs, inputs [None {from=MaterialExpressionConstantBiasScale_0, output=""}], properties {}}
  MaterialExpressionAbs_3 {class=MaterialExpressionAbs, inputs [None {from=MaterialExpressionConstantBiasScale_6, output=""}], properties {}}
  MaterialExpressionAbs_4 {class=MaterialExpressionAbs, inputs [None {from=MaterialExpressionConstantBiasScale_1, output=""}], properties {}}
  MaterialExpressionAbs_5 {class=MaterialExpressionAbs, inputs [None {from=MaterialExpressionConstantBiasScale_2, output=""}], properties {}}
  MaterialExpressionAbs_6 {class=MaterialExpressionAbs, inputs [None {from=MaterialExpressionSubtract_13, output=""}], properties {}}
  MaterialExpressionActorPositionWS_0 {class=MaterialExpressionActorPositionWS, inputs [], properties {}}
  MaterialExpressionActorPositionWS_1 {class=MaterialExpressionActorPositionWS, inputs [], properties {}}
  MaterialExpressionAdd_0 {
   class=MaterialExpressionAdd
   inputs [
    A {from=MaterialExpressionTime_0}
    B {from=MaterialExpressionTextureSampleParameter2D_0, output=B}
   ]
  }
  MaterialExpressionAdd_1 {
   class=MaterialExpressionAdd
   inputs [
    A {from=MaterialExpressionLinearInterpolate_3}
    B {from=MaterialExpressionMultiply_24}
   ]
  }
  MaterialExpressionAdd_10 {
   class=MaterialExpressionAdd
   inputs [
    A {from=MaterialExpressionDivide_6}
    B {from=MaterialExpressionMultiply_18}
   ]
  }
  MaterialExpressionAdd_11 {
   class=MaterialExpressionAdd
   inputs [
    A {from=MaterialExpressionMultiply_11}
    B {from=""}
   ]
  }
  MaterialExpressionAdd_12 {
   class=MaterialExpressionAdd
   inputs [
    A {from=MaterialExpressionDivide_10}
    B {from=MaterialExpressionComponentMask_7}
   ]
  }
  MaterialExpressionAdd_13 {
   class=MaterialExpressionAdd
   inputs [
    A {from=MaterialExpressionDivide_11}
    B {from=MaterialExpressionComponentMask_8}
   ]
  }
  MaterialExpressionAdd_14 {
   class=MaterialExpressionAdd
   inputs [
    A {from=MaterialExpressionOneMinus_0}
    B {from=MaterialExpressionReroute_7}
   ]
  }
  MaterialExpressionAdd_15 {
   class=MaterialExpressionAdd
   inputs [
    A {from=MaterialExpressionOneMinus_4}
    B {from=MaterialExpressionReroute_6}
   ]
  }
  MaterialExpressionAdd_16 {
   class=MaterialExpressionAdd
   inputs [
    A {from=MaterialExpressionOneMinus_5}
    B {from=MaterialExpressionReroute_9}
   ]
  }
  MaterialExpressionAdd_17 {
   class=MaterialExpressionAdd
   inputs [
    A {from=MaterialExpressionMultiply_16}
    B {from=MaterialExpressionMultiply_15}
   ]
  }
  MaterialExpressionAdd_18 {
   class=MaterialExpressionAdd
   inputs [
    A {from=MaterialExpressionAdd_17}
    B {from=MaterialExpressionMultiply_17}
   ]
  }
  MaterialExpressionAdd_19 {
   class=MaterialExpressionAdd
   inputs [
    A {from=MaterialExpressionSubtract_8}
    B {from=MaterialExpressionMultiply_31}
   ]
  }
  MaterialExpressionAdd_2 {
   class=MaterialExpressionAdd
   inputs [
    A {from=MaterialExpressionTime_1}
    B {from=MaterialExpressionMultiply_2}
   ]
  }
  MaterialExpressionAdd_20 {
   class=MaterialExpressionAdd
   inputs [
    A {from=MaterialExpressionMultiply_45}
    B {from=MaterialExpressionMultiply_43}
   ]
  }
  MaterialExpressionAdd_21 {
   class=MaterialExpressionAdd
   inputs [
    A {from=MaterialExpressionConvert_4}
    B {from=MaterialExpressionMultiply_23}
   ]
  }
  MaterialExpressionAdd_22 {
   class=MaterialExpressionAdd
   inputs [
    A {from=MaterialExpressionConvert_1}
    B {from=MaterialExpressionConvert_1}
   ]
  }
  MaterialExpressionAdd_23 {
   class=MaterialExpressionAdd
   inputs [
    A {from=MaterialExpressionNamedRerouteUsage_18}
    B {from=MaterialExpressionNamedRerouteUsage_16}
   ]
  }
  MaterialExpressionAdd_24 {
   class=MaterialExpressionAdd
   inputs [
    A {from=MaterialExpressionConvert_4}
    B {from=MaterialExpressionMultiply_23}
   ]
  }
  MaterialExpressionAdd_25 {
   class=MaterialExpressionAdd
   inputs [
    A {from=MaterialExpressionAdd_21}
    B {from=MaterialExpressionConstant4Vector_0, output=RGBA}
   ]
   properties {constB=0.5}
  }
  MaterialExpressionAdd_26 {
   class=MaterialExpressionAdd
   inputs [
    A {from=MaterialExpressionAdd_24}
    B {from=MaterialExpressionConstant4Vector_0, output=RGBA}
   ]
   properties {constB=0.5}
  }
  MaterialExpressionAdd_27 {
   class=MaterialExpressionAdd
   inputs [
    A {from=MaterialExpressionMultiply_24}
    B {from=MaterialExpressionMultiply_41}
   ]
  }
  MaterialExpressionAdd_28 {
   class=MaterialExpressionAdd
   inputs [
    A {from=MaterialExpressionObjectPositionWS_2}
    B {from=MaterialExpressionAdd_8}
   ]
  }
  MaterialExpressionAdd_29 {
   class=MaterialExpressionAdd
   inputs [
    A {from=MaterialExpressionTextureCoordinate_0}
    B {from=MaterialExpressionMultiply_29}
   ]
  }
  MaterialExpressionAdd_3 {
   class=MaterialExpressionAdd
   inputs [
    A {from=MaterialExpressionWorldPosition_2, output=XYZ}
    B {from=MaterialExpressionMultiply_4}
   ]
  }
  MaterialExpressionAdd_30 {
   class=MaterialExpressionAdd
   inputs [
    A {from=MaterialExpressionSmoothStep_3}
    B {from=MaterialExpressionPower_4}
   ]
  }
  MaterialExpressionAdd_31 {
   class=MaterialExpressionAdd
   inputs [
    A {from=MaterialExpressionOneMinus_6}
    B {from=""}
   ]
  }
  MaterialExpressionAdd_32 {
   class=MaterialExpressionAdd
   inputs [
    A {from=MaterialExpressionMultiply_12}
    B {from=MaterialExpressionReroute_27}
   ]
  }
  MaterialExpressionAdd_33 {
   class=MaterialExpressionAdd
   inputs [
    A {from=MaterialExpressionConvert_0}
    B {from=MaterialExpressionReroute_26}
   ]
  }
  MaterialExpressionAdd_34 {
   class=MaterialExpressionAdd
   inputs [
    A {from=MaterialExpressionScalarParameter_39}
    B {from=MaterialExpressionScalarParameter_38}
   ]
  }
  MaterialExpressionAdd_35 {
   class=MaterialExpressionAdd
   inputs [
    A {from=MaterialExpressionActorPositionWS_1}
    B {from=MaterialExpressionVectorParameter_5, output=RGB}
   ]
  }
  MaterialExpressionAdd_36 {
   class=MaterialExpressionAdd
   inputs [
    A {from=MaterialExpressionAdd_22}
    B {from=MaterialExpressionMultiply_30}
   ]
  }
  MaterialExpressionAdd_37 {
   class=MaterialExpressionAdd
   inputs [
    A {from=MaterialExpressionFloor_6}
    B {from=""}
   ]
   properties {constB=0.5}
  }
  MaterialExpressionAdd_38 {
   class=MaterialExpressionAdd
   inputs [
    A {from=MaterialExpressionFloor_7}
    B {from=""}
   ]
   properties {constB=0.5}
  }
  MaterialExpressionAdd_39 {
   class=MaterialExpressionAdd
   inputs [
    A {from=MaterialExpressionFloor_8}
    B {from=""}
   ]
   properties {constB=0.5}
  }
  MaterialExpressionAdd_4 {
   class=MaterialExpressionAdd
   inputs [
    A {from=MaterialExpressionObjectPositionWS_1}
    B {from=MaterialExpressionAdd_6}
   ]
  }
  MaterialExpressionAdd_40 {
   class=MaterialExpressionAdd
   inputs [
    A {from=MaterialExpressionDivide_24}
    B {from=MaterialExpressionObjectPositionWS_0}
   ]
  }
  MaterialExpressionAdd_41 {
   class=MaterialExpressionAdd
   inputs [
    A {from=MaterialExpressionDivide_23}
    B {from=MaterialExpressionObjectPositionWS_0}
   ]
  }
  MaterialExpressionAdd_42 {
   class=MaterialExpressionAdd
   inputs [
    A {from=MaterialExpressionDivide_22}
    B {from=MaterialExpressionObjectPositionWS_0}
   ]
  }
  MaterialExpressionAdd_43 {
   class=MaterialExpressionAdd
   inputs [
    A {from=MaterialExpressionCustom_0}
    B {from=MaterialExpressionCustom_1}
   ]
  }
  MaterialExpressionAdd_44 {
   class=MaterialExpressionAdd
   inputs [
    A {from=MaterialExpressionAdd_43}
    B {from=MaterialExpressionCustom_2}
   ]
  }
  MaterialExpressionAdd_46 {
   class=MaterialExpressionAdd
   inputs [
    A {from=MaterialExpressionNamedRerouteUsage_19}
    B {from=MaterialExpressionNamedRerouteUsage_18}
   ]
  }
  MaterialExpressionAdd_47 {
   class=MaterialExpressionAdd
   inputs [
    A {from=MaterialExpressionComponentMask_16}
    B {from=MaterialExpressionDivide_25}
   ]
  }
  MaterialExpressionAdd_48 {
   class=MaterialExpressionAdd
   inputs [
    A {from=MaterialExpressionNamedRerouteUsage_22}
    B {from=MaterialExpressionAdd_46}
   ]
  }
  MaterialExpressionAdd_49 {
   class=MaterialExpressionAdd
   inputs [
    A {from=MaterialExpressionMultiply_57}
    B {from=MaterialExpressionComponentMask_18}
   ]
  }
  MaterialExpressionAdd_5 {
   class=MaterialExpressionAdd
   inputs [
    A {from=MaterialExpressionMultiply_9}
    B {from=MaterialExpressionComponentMask_0}
   ]
  }
  MaterialExpressionAdd_50 {
   class=MaterialExpressionAdd
   inputs [
    A {from=MaterialExpressionTextureCoordinate_2}
    B {from=MaterialExpressionMultiply_29}
   ]
  }
  MaterialExpressionAdd_51 {
   class=MaterialExpressionAdd
   inputs [
    A {from=MaterialExpressionComponentMask_20}
    B {from=MaterialExpressionMultiply_59}
   ]
  }
  MaterialExpressionAdd_52 {
   class=MaterialExpressionAdd
   inputs [
    A {from=MaterialExpressionComponentMask_10}
    B {from=MaterialExpressionDivide_26}
   ]
  }
  MaterialExpressionAdd_53 {
   class=MaterialExpressionAdd
   inputs [
    A {from=MaterialExpressionWorldPosition_4, output=XYZ}
    B {from=MaterialExpressionMultiply_54}
   ]
  }
  MaterialExpressionAdd_54 {
   class=MaterialExpressionAdd
   inputs [
    A {from=MaterialExpressionMultiply_55}
    B {from=MaterialExpressionDivide_16}
   ]
  }
  MaterialExpressionAdd_55 {
   class=MaterialExpressionAdd
   inputs [
    A {from=MaterialExpressionAdd_44}
    B {from=MaterialExpressionMultiply_60}
   ]
  }
  MaterialExpressionAdd_56 {
   class=MaterialExpressionAdd
   inputs [
    A {from=MaterialExpressionAdd_26}
    B {from=MaterialExpressionScalarParameter_45}
   ]
  }
  MaterialExpressionAdd_57 {
   class=MaterialExpressionAdd
   inputs [
    A {from=MaterialExpressionDivide_22}
    B {from=MaterialExpressionDivide_23}
   ]
  }
  MaterialExpressionAdd_58 {
   class=MaterialExpressionAdd
   inputs [
    A {from=MaterialExpressionAdd_57}
    B {from=MaterialExpressionDivide_24}
   ]
  }
  MaterialExpressionAdd_59 {
   class=MaterialExpressionAdd
   inputs [
    A {from=MaterialExpressionTransformPosition_0}
    B {from=MaterialExpressionMultiply_54}
   ]
  }
  MaterialExpressionAdd_6 {
   class=MaterialExpressionAdd
   inputs [
    A {from=MaterialExpressionFloor_0}
    B {from=""}
   ]
   properties {constB=0.5}
  }
  MaterialExpressionAdd_7 {
   class=MaterialExpressionAdd
   inputs [
    A {from=MaterialExpressionAdd_0}
    B {from=MaterialExpressionDivide_1}
   ]
  }
  MaterialExpressionAdd_8 {
   class=MaterialExpressionAdd
   inputs [
    A {from=MaterialExpressionNormalize_0}
    B {from=MaterialExpressionAdd_33}
   ]
  }
  MaterialExpressionAdd_9 {
   class=MaterialExpressionAdd
   inputs [
    A {from=MaterialExpressionMultiply_22}
    B {from=MaterialExpressionReroute_20}
   ]
  }
  MaterialExpressionAppendVector_0 {
   class=MaterialExpressionAppendVector
   inputs [
    A {from=MaterialExpressionWorldPosition_7, output=Z}
    B {from=MaterialExpressionComponentMask_12, output=""}
   ]
  }
  MaterialExpressionAppendVector_1 {
   class=MaterialExpressionAppendVector
   inputs [
    A {from=MaterialExpressionFloor_4, output=""}
    B {from=MaterialExpressionMultiply_34, output=""}
   ]
  }
  MaterialExpressionBlendMaterialAttributes_0 {
   class=MaterialExpressionBlendMaterialAttributes
   inputs [
    A {from=MaterialExpressionSetMaterialAttributes_2, output=""}
    B {from=MaterialExpressionMaterialFunctionCall_12, output=MA Out}
    Alpha {from=MaterialExpressionSaturate_0, output=""}
   ]
   properties {vertexAttributeBlendType=UseA}
  }
  MaterialExpressionComponentMask_0 {class=MaterialExpressionComponentMask, inputs [None {from=MaterialExpressionNamedRerouteDeclaration_2}], properties {r=true}}
  MaterialExpressionComponentMask_1 {class=MaterialExpressionComponentMask, inputs [None {from=MaterialExpressionNoise_0}], properties {r=true}}
  MaterialExpressionComponentMask_10 {class=MaterialExpressionComponentMask, inputs [None {from=MaterialExpressionNamedRerouteUsage_26}], properties {a=true}}
  MaterialExpressionComponentMask_11 {class=MaterialExpressionComponentMask, inputs [None {from=MaterialExpressionNamedRerouteUsage_2}], properties {b=true}}
  MaterialExpressionComponentMask_12 {class=MaterialExpressionComponentMask, inputs [None {from=MaterialExpressionNamedRerouteUsage_5}], properties {b=true}}
  MaterialExpressionComponentMask_13 {class=MaterialExpressionComponentMask, inputs [None {from=MaterialExpressionVectorParameter_4, output=RGB}], properties {g=true, r=true}}
  MaterialExpressionComponentMask_14 {class=MaterialExpressionComponentMask, inputs [None {from=MaterialExpressionNamedRerouteUsage_3}], properties {g=true}}
  MaterialExpressionComponentMask_15 {class=MaterialExpressionComponentMask, inputs [None {from=MaterialExpressionNamedRerouteUsage_26}], properties {g=true, r=true}}
  MaterialExpressionComponentMask_16 {class=MaterialExpressionComponentMask, inputs [None {from=MaterialExpressionNamedRerouteUsage_26}], properties {b=true}}
  MaterialExpressionComponentMask_17 {class=MaterialExpressionComponentMask, inputs [None {from=MaterialExpressionConstantBiasScale_5}], properties {b=true, g=true, r=true}}
  MaterialExpressionComponentMask_18 {class=MaterialExpressionComponentMask, inputs [None {from=MaterialExpressionNamedRerouteUsage_11}], properties {g=true}}
  MaterialExpressionComponentMask_19 {class=MaterialExpressionComponentMask, inputs [None {from=MaterialExpressionNamedRerouteUsage_21}], properties {a=true}}
  MaterialExpressionComponentMask_2 {
   class=MaterialExpressionComponentMask
   inputs [None {from=MaterialExpressionMaterialFunctionCall_23, output=packedDXY}]
   properties {g=true, r=true}
  }
  MaterialExpressionComponentMask_20 {class=MaterialExpressionComponentMask, inputs [None {from=MaterialExpressionNamedRerouteUsage_27}], properties {b=true}}
  MaterialExpressionComponentMask_22 {class=MaterialExpressionComponentMask, inputs [None {from=MaterialExpressionDivide_2}], properties {g=true, r=true}}
  MaterialExpressionComponentMask_23 {class=MaterialExpressionComponentMask, inputs [None {from=MaterialExpressionDivide_2}], properties {b=true, g=true}}
  MaterialExpressionComponentMask_24 {class=MaterialExpressionComponentMask, inputs [None {from=MaterialExpressionDivide_2}], properties {b=true, r=true}}
  MaterialExpressionComponentMask_25 {class=MaterialExpressionComponentMask, inputs [None {from=MaterialExpressionNamedRerouteUsage_29}], properties {a=true}}
  MaterialExpressionComponentMask_3 {
   class=MaterialExpressionComponentMask
   inputs [None {from=MaterialExpressionMaterialFunctionCall_23, output=packedDXY}]
   properties {a=true, b=true}
  }
  MaterialExpressionComponentMask_4 {class=MaterialExpressionComponentMask, inputs [None {from=MaterialExpressionAbs_1}], properties {b=true}}
  MaterialExpressionComponentMask_5 {class=MaterialExpressionComponentMask, inputs [None {from=MaterialExpressionAbs_1}], properties {r=true}}
  MaterialExpressionComponentMask_6 {class=MaterialExpressionComponentMask, inputs [None {from=MaterialExpressionWorldPosition_0, output=XYZ}], properties {g=true, r=true}}
  MaterialExpressionComponentMask_7 {class=MaterialExpressionComponentMask, inputs [None {from=MaterialExpressionFrac_3}], properties {g=true, r=true}}
  MaterialExpressionComponentMask_8 {class=MaterialExpressionComponentMask, inputs [None {from=MaterialExpressionFrac_3}], properties {a=true, b=true}}
  MaterialExpressionComponentMask_9 {class=MaterialExpressionComponentMask, inputs [None {from=MaterialExpressionAbs_1}], properties {g=true}}
  MaterialExpressionComposite_0 {
   class=MaterialExpressionComposite
   inputs [None {from=MaterialExpressionMaterialFunctionCall_9, output=Result}]
   properties {desc=Transform to World in case we will use RVT after this point, subgraphName=Tangent to World}
  }
  MaterialExpressionComposite_1 {
   class=MaterialExpressionComposite
   inputs [MaterialAttributes {from=MaterialExpressionMaterialFunctionCall_30, output=Result}]
   properties {subgraphName=Documentation Link}
  }
  MaterialExpressionComposite_2 {
   class=MaterialExpressionComposite
   inputs [MaterialAttributes {from=MaterialExpressionReroute_71, output=Output}]
   properties {desc=Game Shared WPO and Fire Effect, subgraphName=WPO And Shared Functions}
  }
  MaterialExpressionComposite_4 {
   class=MaterialExpressionComposite
   inputs [InputPin {from=MaterialExpressionMaterialFunctionCall_17, output=Result}]
   properties {subgraphName=Mobile Opt}
  }
  MaterialExpressionComposite_7 {
   class=MaterialExpressionComposite
   inputs [
    "For Moss (MA)" {from=MaterialExpressionMaterialFunctionCall_6, output=Result}
    "RVT_Height (S)" {from=MaterialExpressionMaterialFunctionCall_17, output=RVT H}
   ]
   properties {subgraphName="Seasonal "}
  }
  MaterialExpressionConstant3Vector_2 {class=MaterialExpressionConstant3Vector, properties {constant=<rgba 0 1 1 0>}}
  MaterialExpressionConstant4Vector_0 {class=MaterialExpressionConstant4Vector, inputs [], properties {constant=<rgba 0 0 0.5 0.5>}}
  MaterialExpressionConstantBiasScale_0 {class=MaterialExpressionConstantBiasScale, inputs [None {from=MaterialExpressionAdd_1}], properties {bias=-0.5, scale=2}}
  MaterialExpressionConstantBiasScale_1 {class=MaterialExpressionConstantBiasScale, inputs [None {from=MaterialExpressionFrac_5}], properties {bias=-0.5, scale=2}}
  MaterialExpressionConstantBiasScale_2 {class=MaterialExpressionConstantBiasScale, inputs [None {from=MaterialExpressionFrac_7}], properties {bias=-0.5, scale=2}}
  MaterialExpressionConstantBiasScale_3 {class=MaterialExpressionConstantBiasScale, inputs [None {from=MaterialExpressionFrac_0}], properties {bias=-0.5, scale=2}}
  MaterialExpressionConstantBiasScale_4 {class=MaterialExpressionConstantBiasScale, inputs [None {from=MaterialExpressionFrac_10}], properties {bias=0.05, scale=0.9}}
  MaterialExpressionConstantBiasScale_5 {class=MaterialExpressionConstantBiasScale, inputs [None {from=MaterialExpressionFrac_11}], properties {bias=-0.5, scale=2}}
  MaterialExpressionConstantBiasScale_6 {class=MaterialExpressionConstantBiasScale, inputs [None {from=MaterialExpressionSaturate_4}], properties {bias=-0.5, scale=2}}
  MaterialExpressionConstantBiasScale_7 {class=MaterialExpressionConstantBiasScale, inputs [None {from=MaterialExpressionSine_3}]}
  MaterialExpressionConstant_0 {class=MaterialExpressionConstant}
  MaterialExpressionConstant_1 {class=MaterialExpressionConstant, properties {r=1}}
  MaterialExpressionConstant_2 {class=MaterialExpressionConstant, properties {r=0.8}}
  MaterialExpressionConstant_3 {class=MaterialExpressionConstant, properties {r=200}}
  MaterialExpressionConstant_4 {class=MaterialExpressionConstant, properties {r=0.5}}
  MaterialExpressionConstant_5 {class=MaterialExpressionConstant, properties {r=1}}
  MaterialExpressionConstant_6 {class=MaterialExpressionConstant, properties {r=1.618}}
  MaterialExpressionConstant_7 {class=MaterialExpressionConstant}
  MaterialExpressionConstant_8 {class=MaterialExpressionConstant, properties {r=1.3247}}
  MaterialExpressionConstant_9 {class=MaterialExpressionConstant, properties {r=3.1416}}
  MaterialExpressionConvert_0 {
   class=MaterialExpressionConvert
   inputs [
    None {from=""}
    None {from=MaterialExpressionAdd_32}
    None {from=""}
   ]
   properties {
    nodeName=Make Vector3
    convertInputs [
     {type=Scalar, expressionInput {expression=null, inputName=None, mask=0, maskA=0, maskB=0, maskG=0, maskR=0, outputIndex=0}}
     {
      type=Scalar
      expressionInput {
       expression=&self:MaterialExpressionAdd_32
       inputName=None
       mask=0
       maskA=0
       maskB=0
       maskG=0
       maskR=0
       outputIndex=0
      }
     }
     {type=Scalar, expressionInput {expression=null, inputName=None, mask=0, maskA=0, maskB=0, maskG=0, maskR=0, outputIndex=0}}
    ]
    convertOutputs [{type=Vector3}]
   }
  }
  MaterialExpressionConvert_1 {
   class=MaterialExpressionConvert
   inputs [None {from=MaterialExpressionDivide_15}]
   properties {
    nodeName=Break Float2
    convertInputs [{
      type=Vector2
      expressionInput {
       expression=&self:MaterialExpressionDivide_15
       inputName=None
       mask=0
       maskA=0
       maskB=0
       maskG=0
       maskR=0
       outputIndex=0
      }
     }]
    convertOutputs [{type=Scalar}, {type=Scalar}]
   }
  }
  MaterialExpressionConvert_10 {
   class=MaterialExpressionConvert
   inputs [
    None {from=MaterialExpressionScalarParameter_69}
    None {from=MaterialExpressionScalarParameter_70}
    None {from=MaterialExpressionScalarParameter_69}
    None {from=MaterialExpressionScalarParameter_70}
   ]
   properties {
    nodeName=Make Vector4
    convertInputs [
     {
      type=Scalar
      expressionInput {
       expression=&self:MaterialExpressionScalarParameter_69
       inputName=None
       mask=0
       maskA=0
       maskB=0
       maskG=0
       maskR=0
       outputIndex=0
      }
     }
     {
      type=Scalar
      expressionInput {
       expression=&self:MaterialExpressionScalarParameter_70
       inputName=None
       mask=0
       maskA=0
       maskB=0
       maskG=0
       maskR=0
       outputIndex=0
      }
     }
     {
      type=Scalar
      expressionInput {
       expression=&self:MaterialExpressionScalarParameter_69
       inputName=None
       mask=0
       maskA=0
       maskB=0
       maskG=0
       maskR=0
       outputIndex=0
      }
     }
     {
      type=Scalar
      expressionInput {
       expression=&self:MaterialExpressionScalarParameter_70
       inputName=None
       mask=0
       maskA=0
       maskB=0
       maskG=0
       maskR=0
       outputIndex=0
      }
     }
    ]
    convertOutputs [{type=Vector4, defaultValue=<rgba 0 0 0 1>}]
   }
  }
  MaterialExpressionConvert_2 {
   class=MaterialExpressionConvert
   inputs [None {from=MaterialExpressionTransform_2}]
   properties {
    nodeName=Swizzle Float3
    convertInputs [{
      type=Vector3
      expressionInput {
       expression=&self:MaterialExpressionTransform_2
       inputName=None
       mask=0
       maskA=0
       maskB=0
       maskG=0
       maskR=0
       outputIndex=0
      }
     }]
    convertOutputs [{type=Vector3}]
   }
  }
  MaterialExpressionConvert_3 {
   class=MaterialExpressionConvert
   inputs [
    None {from=MaterialExpressionConstant_6}
    None {from=MaterialExpressionConstant_8}
    None {from=MaterialExpressionConstant_9}
   ]
   properties {
    nodeName=Make Vector3
    convertInputs [
     {
      type=Scalar
      expressionInput {
       expression=&self:MaterialExpressionConstant_6
       inputName=None
       mask=0
       maskA=0
       maskB=0
       maskG=0
       maskR=0
       outputIndex=0
      }
     }
     {
      type=Scalar
      expressionInput {
       expression=&self:MaterialExpressionConstant_8
       inputName=None
       mask=0
       maskA=0
       maskB=0
       maskG=0
       maskR=0
       outputIndex=0
      }
     }
     {
      type=Scalar
      expressionInput {
       expression=&self:MaterialExpressionConstant_9
       inputName=None
       mask=0
       maskA=0
       maskB=0
       maskG=0
       maskR=0
       outputIndex=0
      }
     }
    ]
    convertOutputs [{type=Vector3}]
   }
  }
  MaterialExpressionConvert_4 {
   class=MaterialExpressionConvert
   inputs [None {from=MaterialExpressionSubtract_6}]
   properties {
    nodeName=Break Float2
    convertInputs [{
      type=Vector2
      expressionInput {
       expression=&self:MaterialExpressionSubtract_6
       inputName=None
       mask=0
       maskA=0
       maskB=0
       maskG=0
       maskR=0
       outputIndex=0
      }
     }]
    convertOutputs [{type=Scalar}, {type=Scalar}]
   }
  }
  MaterialExpressionConvert_5 {
   class=MaterialExpressionConvert
   inputs [
    None {from=MaterialExpressionConstant_9}
    None {from=MaterialExpressionConstant_6}
    None {from=MaterialExpressionConstant_8}
   ]
   properties {
    nodeName=Make Vector3
    convertInputs [
     {
      type=Scalar
      expressionInput {
       expression=&self:MaterialExpressionConstant_9
       inputName=None
       mask=0
       maskA=0
       maskB=0
       maskG=0
       maskR=0
       outputIndex=0
      }
     }
     {
      type=Scalar
      expressionInput {
       expression=&self:MaterialExpressionConstant_6
       inputName=None
       mask=0
       maskA=0
       maskB=0
       maskG=0
       maskR=0
       outputIndex=0
      }
     }
     {
      type=Scalar
      expressionInput {
       expression=&self:MaterialExpressionConstant_8
       inputName=None
       mask=0
       maskA=0
       maskB=0
       maskG=0
       maskR=0
       outputIndex=0
      }
     }
    ]
    convertOutputs [{type=Vector3}]
   }
  }
  MaterialExpressionConvert_6 {
   class=MaterialExpressionConvert
   inputs [None {from=MaterialExpressionSmoothStep_1}]
   properties {
    nodeName=Break Float4
    convertInputs [{
      type=Vector4
      defaultValue=<rgba 0 0 0 1>
      expressionInput {
       expression=&self:MaterialExpressionSmoothStep_1
       inputName=None
       mask=0
       maskA=0
       maskB=0
       maskG=0
       maskR=0
       outputIndex=0
      }
     }]
    convertOutputs [{type=Scalar}, {type=Scalar}, {type=Scalar}, {type=Scalar}]
   }
  }
  MaterialExpressionConvert_7 {
   class=MaterialExpressionConvert
   inputs [None {from=MaterialExpressionSmoothStep_2}]
   properties {
    nodeName=Break Float4
    convertInputs [{
      type=Vector4
      defaultValue=<rgba 0 0 0 1>
      expressionInput {
       expression=&self:MaterialExpressionSmoothStep_2
       inputName=None
       mask=0
       maskA=0
       maskB=0
       maskG=0
       maskR=0
       outputIndex=0
      }
     }]
    convertOutputs [{type=Scalar}, {type=Scalar}, {type=Scalar}, {type=Scalar}]
   }
  }
  MaterialExpressionConvert_8 {
   class=MaterialExpressionConvert
   inputs [
    None {from=MaterialExpressionScalarParameter_30}
    None {from=MaterialExpressionScalarParameter_31}
    None {from=MaterialExpressionScalarParameter_30}
    None {from=MaterialExpressionScalarParameter_31}
   ]
   properties {
    nodeName=Make Vector4
    convertInputs [
     {
      type=Scalar
      expressionInput {
       expression=&self:MaterialExpressionScalarParameter_30
       inputName=None
       mask=0
       maskA=0
       maskB=0
       maskG=0
       maskR=0
       outputIndex=0
      }
     }
     {
      type=Scalar
      expressionInput {
       expression=&self:MaterialExpressionScalarParameter_31
       inputName=None
       mask=0
       maskA=0
       maskB=0
       maskG=0
       maskR=0
       outputIndex=0
      }
     }
     {
      type=Scalar
      expressionInput {
       expression=&self:MaterialExpressionScalarParameter_30
       inputName=None
       mask=0
       maskA=0
       maskB=0
       maskG=0
       maskR=0
       outputIndex=0
      }
     }
     {
      type=Scalar
      expressionInput {
       expression=&self:MaterialExpressionScalarParameter_31
       inputName=None
       mask=0
       maskA=0
       maskB=0
       maskG=0
       maskR=0
       outputIndex=0
      }
     }
    ]
    convertOutputs [{type=Vector4, defaultValue=<rgba 0 0 0 1>}]
   }
  }
  MaterialExpressionConvert_9 {
   class=MaterialExpressionConvert
   inputs [
    None {from=MaterialExpressionConstant_8}
    None {from=MaterialExpressionConstant_9}
    None {from=MaterialExpressionConstant_6}
   ]
   properties {
    nodeName=Make Vector3
    convertInputs [
     {
      type=Scalar
      expressionInput {
       expression=&self:MaterialExpressionConstant_8
       inputName=None
       mask=0
       maskA=0
       maskB=0
       maskG=0
       maskR=0
       outputIndex=0
      }
     }
     {
      type=Scalar
      expressionInput {
       expression=&self:MaterialExpressionConstant_9
       inputName=None
       mask=0
       maskA=0
       maskB=0
       maskG=0
       maskR=0
       outputIndex=0
      }
     }
     {
      type=Scalar
      expressionInput {
       expression=&self:MaterialExpressionConstant_6
       inputName=None
       mask=0
       maskA=0
       maskB=0
       maskG=0
       maskR=0
       outputIndex=0
      }
     }
    ]
    convertOutputs [{type=Vector3}]
   }
  }
  MaterialExpressionCustom_0 {
   class=MaterialExpressionCustom
   inputs [
    Pos {from=MaterialExpressionAdd_42, output=""}
    T {from=MaterialExpressionTime_17, output=""}
   ]
   properties {
    code="// The below expression will get compiled\r\n// into the output of this node\r\nreturn Rand32ToFloat(Rand4DPCG32(int4(floor(Pos), T)));"
    description=Vector4 Noise
    outputType=CMOT_Float4
    inputs [
     {
      inputName=Pos
      input {
       expression=&self:MaterialExpressionAdd_42
       inputName=None
       mask=0
       maskA=0
       maskB=0
       maskG=0
       maskR=0
       outputIndex=0
      }
     }
     {
      inputName=T
      input {
       expression=&self:MaterialExpressionTime_17
       inputName=None
       mask=0
       maskA=0
       maskB=0
       maskG=0
       maskR=0
       outputIndex=0
      }
     }
    ]
   }
  }
  MaterialExpressionCustom_1 {
   class=MaterialExpressionCustom
   inputs [
    Pos {from=MaterialExpressionAdd_41, output=""}
    T {from=MaterialExpressionTime_17, output=""}
   ]
   properties {
    code="// The below expression will get compiled\r\n// into the output of this node\r\nreturn Rand32ToFloat(Rand4DPCG32(int4(floor(Pos), T)));"
    description=Vector4 Noise
    outputType=CMOT_Float4
    inputs [
     {
      inputName=Pos
      input {
       expression=&self:MaterialExpressionAdd_41
       inputName=None
       mask=0
       maskA=0
       maskB=0
       maskG=0
       maskR=0
       outputIndex=0
      }
     }
     {
      inputName=T
      input {
       expression=&self:MaterialExpressionTime_17
       inputName=None
       mask=0
       maskA=0
       maskB=0
       maskG=0
       maskR=0
       outputIndex=0
      }
     }
    ]
   }
  }
  MaterialExpressionCustom_2 {
   class=MaterialExpressionCustom
   inputs [
    Pos {from=MaterialExpressionAdd_40, output=""}
    T {from=MaterialExpressionTime_17, output=""}
   ]
   properties {
    code="// The below expression will get compiled\r\n// into the output of this node\r\nreturn Rand32ToFloat(Rand4DPCG32(int4(floor(Pos), T)));"
    description=Vector4 Noise
    outputType=CMOT_Float4
    inputs [
     {
      inputName=Pos
      input {
       expression=&self:MaterialExpressionAdd_40
       inputName=None
       mask=0
       maskA=0
       maskB=0
       maskG=0
       maskR=0
       outputIndex=0
      }
     }
     {
      inputName=T
      input {
       expression=&self:MaterialExpressionTime_17
       inputName=None
       mask=0
       maskA=0
       maskB=0
       maskG=0
       maskR=0
       outputIndex=0
      }
     }
    ]
   }
  }
  MaterialExpressionDivide_0 {
   class=MaterialExpressionDivide
   inputs [
    A {from=MaterialExpressionMultiply_7}
    B {from=MaterialExpressionVectorParameter_19, output=RGB}
   ]
  }
  MaterialExpressionDivide_1 {
   class=MaterialExpressionDivide
   inputs [
    A {from=MaterialExpressionWorldPosition_6, output=Z}
    B {from=""}
   ]
   properties {constB=-900}
  }
  MaterialExpressionDivide_10 {
   class=MaterialExpressionDivide
   inputs [
    A {from=MaterialExpressionComponentMask_6}
    B {from=MaterialExpressionScalarParameter_22}
   ]
  }
  MaterialExpressionDivide_11 {
   class=MaterialExpressionDivide
   inputs [
    A {from=MaterialExpressionComponentMask_6}
    B {from=MaterialExpressionScalarParameter_23}
   ]
  }
  MaterialExpressionDivide_12 {
   class=MaterialExpressionDivide
   inputs [
    A {from=MaterialExpressionAppendVector_0}
    B {from=MaterialExpressionScalarParameter_28}
   ]
  }
  MaterialExpressionDivide_13 {
   class=MaterialExpressionDivide
   inputs [
    A {from=MaterialExpressionSubtract_11}
    B {from=MaterialExpressionScalarParameter_40}
   ]
  }
  MaterialExpressionDivide_14 {
   class=MaterialExpressionDivide
   inputs [
    A {from=MaterialExpressionMultiply_68}
    B {from=""}
   ]
   properties {constB=3}
  }
  MaterialExpressionDivide_15 {
   class=MaterialExpressionDivide
   inputs [
    A {from=MaterialExpressionActorPositionWS_0}
    B {from=""}
   ]
   properties {constB=16180.2988}
  }
  MaterialExpressionDivide_16 {
   class=MaterialExpressionDivide
   inputs [
    A {from=MaterialExpressionMultiply_14}
    B {from=MaterialExpressionMax_13}
   ]
  }
  MaterialExpressionDivide_17 {
   class=MaterialExpressionDivide
   inputs [
    A {from=MaterialExpressionMultiply_8}
    B {from=MaterialExpressionMaterialFunctionCall_19, output=Scale XYZ}
   ]
  }
  MaterialExpressionDivide_18 {
   class=MaterialExpressionDivide
   inputs [
    A {from=MaterialExpressionSubtract_15}
    B {from=MaterialExpressionOneMinus_10}
   ]
  }
  MaterialExpressionDivide_19 {
   class=MaterialExpressionDivide
   inputs [
    A {from=MaterialExpressionDivide_0}
    B {from=MaterialExpressionConvert_3}
   ]
  }
  MaterialExpressionDivide_2 {
   class=MaterialExpressionDivide
   inputs [
    A {from=MaterialExpressionAdd_3}
    B {from=""}
   ]
   properties {constB=1600}
  }
  MaterialExpressionDivide_20 {
   class=MaterialExpressionDivide
   inputs [
    A {from=MaterialExpressionDivide_0}
    B {from=MaterialExpressionConvert_5}
   ]
  }
  MaterialExpressionDivide_21 {
   class=MaterialExpressionDivide
   inputs [
    A {from=MaterialExpressionDivide_0}
    B {from=MaterialExpressionConvert_9}
   ]
  }
  MaterialExpressionDivide_22 {
   class=MaterialExpressionDivide
   inputs [
    A {from=MaterialExpressionMultiply_47}
    B {from=MaterialExpressionMaterialFunctionCall_19, output=Scale XYZ}
   ]
  }
  MaterialExpressionDivide_23 {
   class=MaterialExpressionDivide
   inputs [
    A {from=MaterialExpressionMultiply_48}
    B {from=MaterialExpressionMaterialFunctionCall_19, output=Scale XYZ}
   ]
  }
  MaterialExpressionDivide_24 {
   class=MaterialExpressionDivide
   inputs [
    A {from=MaterialExpressionMultiply_49}
    B {from=MaterialExpressionMaterialFunctionCall_19, output=Scale XYZ}
   ]
  }
  MaterialExpressionDivide_25 {
   class=MaterialExpressionDivide
   inputs [
    A {from=MaterialExpressionFloor_9}
    B {from=MaterialExpressionScalarParameter_42}
   ]
  }
  MaterialExpressionDivide_26 {
   class=MaterialExpressionDivide
   inputs [
    A {from=MaterialExpressionFloor_5}
    B {from=""}
   ]
   properties {constB=0.013}
  }
  MaterialExpressionDivide_27 {
   class=MaterialExpressionDivide
   inputs [
    A {from=MaterialExpressionDivide_14}
    B {from=MaterialExpressionScalarParameter_68}
   ]
  }
  MaterialExpressionDivide_3 {
   class=MaterialExpressionDivide
   inputs [
    A {from=MaterialExpressionLength_0}
    B {from=MaterialExpressionScalarParameter_12}
   ]
  }
  MaterialExpressionDivide_4 {
   class=MaterialExpressionDivide
   inputs [
    A {from=MaterialExpressionSquareRoot_0}
    B {from=MaterialExpressionVectorParameter_16, output=A}
   ]
  }
  MaterialExpressionDivide_5 {
   class=MaterialExpressionDivide
   inputs [
    A {from=""}
    B {from=MaterialExpressionReroute_21}
   ]
  }
  MaterialExpressionDivide_6 {
   class=MaterialExpressionDivide
   inputs [
    A {from=MaterialExpressionConvert_2}
    B {from=MaterialExpressionVectorParameter_7, output=WiggleWidth}
   ]
  }
  MaterialExpressionDivide_7 {
   class=MaterialExpressionDivide
   inputs [
    A {from=""}
    B {from=MaterialExpressionScalarParameter_7}
   ]
  }
  MaterialExpressionDivide_8 {
   class=MaterialExpressionDivide
   inputs [
    A {from=MaterialExpressionSubtract_3}
    B {from=MaterialExpressionScalarParameter_6}
   ]
   properties {constB=800}
  }
  MaterialExpressionDivide_9 {
   class=MaterialExpressionDivide
   inputs [
    A {from=MaterialExpressionSubtract_4}
    B {from=MaterialExpressionScalarParameter_10}
   ]
   properties {constB=0.5}
  }
  MaterialExpressionDotProduct_0 {
   class=MaterialExpressionDotProduct
   inputs [
    A {from=MaterialExpressionSubtract_16, output=""}
    B {from=MaterialExpressionNormalize_2, output=""}
   ]
   properties {}
  }
  MaterialExpressionDotProduct_1 {
   class=MaterialExpressionDotProduct
   inputs [
    A {from=MaterialExpressionSubtract_17, output=""}
    B {from=MaterialExpressionSubtract_17, output=""}
   ]
   properties {}
  }
  MaterialExpressionFloor_0 {class=MaterialExpressionFloor, inputs [None {from=MaterialExpressionDivide_0, output=""}], properties {}}
  MaterialExpressionFloor_1 {class=MaterialExpressionFloor, inputs [None {from=MaterialExpressionAdd_14, output=""}], properties {}}
  MaterialExpressionFloor_2 {class=MaterialExpressionFloor, inputs [None {from=MaterialExpressionAdd_15, output=""}], properties {}}
  MaterialExpressionFloor_3 {class=MaterialExpressionFloor, inputs [None {from=MaterialExpressionAdd_16, output=""}], properties {}}
  MaterialExpressionFloor_4 {class=MaterialExpressionFloor, inputs [None {from=MaterialExpressionMultiply_21, output=""}], properties {}}
  MaterialExpressionFloor_5 {class=MaterialExpressionFloor, inputs [None {from=MaterialExpressionMultiply_52, output=""}], properties {}}
  MaterialExpressionFloor_6 {class=MaterialExpressionFloor, inputs [None {from=MaterialExpressionDivide_19, output=""}], properties {}}
  MaterialExpressionFloor_7 {class=MaterialExpressionFloor, inputs [None {from=MaterialExpressionDivide_20, output=""}], properties {}}
  MaterialExpressionFloor_8 {class=MaterialExpressionFloor, inputs [None {from=MaterialExpressionDivide_21, output=""}], properties {}}
  MaterialExpressionFloor_9 {class=MaterialExpressionFloor, inputs [None {from=MaterialExpressionMultiply_50, output=""}], properties {}}
  MaterialExpressionFrac_0 {class=MaterialExpressionFrac, inputs [None {from=MaterialExpressionMin_4, output=""}], properties {}}
  MaterialExpressionFrac_1 {class=MaterialExpressionFrac, inputs [None {from=MaterialExpressionAdd_1, output=""}], properties {}}
  MaterialExpressionFrac_10 {class=MaterialExpressionFrac, inputs [None {from=MaterialExpressionAdd_51, output=""}], properties {}}
  MaterialExpressionFrac_11 {class=MaterialExpressionFrac, inputs [None {from=MaterialExpressionAdd_55, output=""}], properties {}}
  MaterialExpressionFrac_12 {class=MaterialExpressionFrac, inputs [None {from=MaterialExpressionComponentMask_25, output=""}], properties {}}
  MaterialExpressionFrac_14 {class=MaterialExpressionFrac, inputs [None {from=MaterialExpressionDivide_27, output=""}], properties {}}
  MaterialExpressionFrac_2 {class=MaterialExpressionFrac, inputs [None {from=MaterialExpressionAdd_5, output=""}], properties {}}
  MaterialExpressionFrac_3 {class=MaterialExpressionFrac, inputs [None {from=MaterialExpressionMultiply_13, output=""}], properties {}}
  MaterialExpressionFrac_4 {class=MaterialExpressionFrac, inputs [None {from=MaterialExpressionAdd_47, output=""}], properties {}}
  MaterialExpressionFrac_5 {class=MaterialExpressionFrac, inputs [None {from=MaterialExpressionAdd_25, output=""}], properties {}}
  MaterialExpressionFrac_6 {class=MaterialExpressionFrac, inputs [None {from=MaterialExpressionMultiply_26, output=""}], properties {}}
  MaterialExpressionFrac_7 {class=MaterialExpressionFrac, inputs [None {from=MaterialExpressionAdd_56, output=""}], properties {}}
  MaterialExpressionFrac_8 {class=MaterialExpressionFrac, inputs [None {from=MaterialExpressionAdd_52, output=""}], properties {}}
  MaterialExpressionFrac_9 {class=MaterialExpressionFrac, inputs [None {from=MaterialExpressionAdd_49, output=""}], properties {}}
  MaterialExpressionGetMaterialAttributes_0 {
   class=MaterialExpressionGetMaterialAttributes
   inputs [None {from=MaterialExpressionReroute_3, output=Input}]
   properties {attributeGetTypes [#1]}
  }
  MaterialExpressionGetMaterialAttributes_1 {class=MaterialExpressionGetMaterialAttributes, inputs [None {from=MaterialExpressionReroute_0, output=InputPin}]}
  MaterialExpressionGetMaterialAttributes_3 {
   class=MaterialExpressionGetMaterialAttributes
   inputs [None {from=MaterialExpressionMaterialFunctionCall_6, output=Result}]
   properties {attributeGetTypes [#2]}
  }
  MaterialExpressionIf_0 {
   class=MaterialExpressionIf
   inputs [
    A {from=MaterialExpressionFrac_2, output=""}
    B {from=MaterialExpressionScalarParameter_13, output=""}
    "A > B" {from=MaterialExpressionTransformPosition_2, output=""}
    "A == B" {from="", output=""}
    "A < B" {from=MaterialExpressionWorldPosition_4, output=XYZ}
   ]
   properties {constB=0.75}
  }
  MaterialExpressionIf_1 {
   class=MaterialExpressionIf
   inputs [
    A {from=MaterialExpressionScalarParameter_32, output=""}
    B {from="", output=""}
    "A > B" {from=MaterialExpressionOneMinus_8, output=""}
    "A == B" {from="", output=""}
    "A < B" {from=MaterialExpressionMax_12, output=""}
   ]
   properties {constB=0.5}
  }
  MaterialExpressionIf_10 {
   class=MaterialExpressionIf
   inputs [
    A {from=MaterialExpressionScalarParameter_60, output=""}
    B {from="", output=""}
    "A > B" {from=MaterialExpressionNamedRerouteUsage_32, output=""}
    "A == B" {from="", output=""}
    "A < B" {from=MaterialExpressionOneMinus_9, output=""}
   ]
   properties {constB=0.5}
  }
  MaterialExpressionIf_2 {
   class=MaterialExpressionIf
   inputs [
    A {from=MaterialExpressionSaturate_8, output=""}
    B {from="", output=""}
    "A > B" {from=MaterialExpressionMultiply_14, output=""}
    "A == B" {from="", output=""}
    "A < B" {from=MaterialExpressionMultiply_38, output=""}
   ]
   properties {constB=0.5}
  }
  MaterialExpressionIf_3 {
   class=MaterialExpressionIf
   inputs [
    A {from=MaterialExpressionTwoSidedSign_4, output=""}
    B {from="", output=""}
    "A > B" {from=MaterialExpressionScalarParameter_35, output=""}
    "A == B" {from="", output=""}
    "A < B" {from=MaterialExpressionScalarParameter_36, output=""}
   ]
   properties {}
  }
  MaterialExpressionIf_4 {
   class=MaterialExpressionIf
   inputs [
    A {from=MaterialExpressionScalarParameter_37, output=""}
    B {from="", output=""}
    "A > B" {from=MaterialExpressionNamedRerouteUsage_17, output=""}
    "A == B" {from="", output=""}
    "A < B" {from=MaterialExpressionMax_18, output=""}
   ]
   properties {constB=0.5}
  }
  MaterialExpressionIf_5 {
   class=MaterialExpressionIf
   inputs [
    A {from=MaterialExpressionScalarParameter_49, output=""}
    B {from="", output=""}
    "A > B" {from=MaterialExpressionScalarParameter_57, output=""}
    "A == B" {from="", output=""}
    "A < B" {from=MaterialExpressionScalarParameter_50, output=""}
   ]
   properties {constB=0.5}
  }
  MaterialExpressionIf_6 {
   class=MaterialExpressionIf
   inputs [
    A {from=MaterialExpressionScalarParameter_49, output=""}
    B {from="", output=""}
    "A > B" {from=MaterialExpressionScalarParameter_55, output=""}
    "A == B" {from="", output=""}
    "A < B" {from=MaterialExpressionScalarParameter_56, output=""}
   ]
   properties {constB=0.5}
  }
  MaterialExpressionIf_7 {
   class=MaterialExpressionIf
   inputs [
    A {from=MaterialExpressionScalarParameter_49, output=""}
    B {from="", output=""}
    "A > B" {from=MaterialExpressionScalarParameter_58, output=""}
    "A == B" {from="", output=""}
    "A < B" {from=MaterialExpressionScalarParameter_59, output=""}
   ]
   properties {constB=0.5}
  }
  MaterialExpressionIf_8 {
   class=MaterialExpressionIf
   inputs [
    A {from=MaterialExpressionMultiply_58, output=""}
    B {from=MaterialExpressionScalarParameter_65, output=""}
    "A > B" {from=MaterialExpressionDivide_18, output=""}
    "A == B" {from="", output=""}
    "A < B" {from=MaterialExpressionConstant_7, output=""}
   ]
   properties {}
  }
  MaterialExpressionIf_9 {
   class=MaterialExpressionIf
   inputs [
    A {from=MaterialExpressionScalarParameter_49, output=""}
    B {from="", output=""}
    "A > B" {from=MaterialExpressionScalarParameter_57, output=""}
    "A == B" {from="", output=""}
    "A < B" {from=MaterialExpressionScalarParameter_56, output=""}
   ]
   properties {constB=0.5}
  }
  MaterialExpressionLength_0 {class=MaterialExpressionLength, inputs [None {from=MaterialExpressionSubtract_0, output=""}], properties {}}
  MaterialExpressionLinearInterpolate_0 {
   class=MaterialExpressionLinearInterpolate
   inputs [
    A {from=MaterialExpressionOneMinus_2}
    B {from=MaterialExpressionPower_1}
    Alpha {from=MaterialExpressionSaturate_5}
   ]
  }
  MaterialExpressionLinearInterpolate_1 {
   class=MaterialExpressionLinearInterpolate
   inputs [
    A {from=MaterialExpressionMax_6}
    B {from=MaterialExpressionOneMinus_3}
    Alpha {from=MaterialExpressionScalarParameter_0}
   ]
  }
  MaterialExpressionLinearInterpolate_10 {
   class=MaterialExpressionLinearInterpolate
   inputs [
    A {from=MaterialExpressionTextureCoordinate_2}
    B {from=MaterialExpressionAdd_50}
    Alpha {from=MaterialExpressionLinearInterpolate_11}
   ]
  }
  MaterialExpressionLinearInterpolate_11 {
   class=MaterialExpressionLinearInterpolate
   inputs [
    A {from=MaterialExpressionMultiply_35}
    B {from=""}
    Alpha {from=MaterialExpressionMultiply_67}
   ]
  }
  MaterialExpressionLinearInterpolate_2 {
   class=MaterialExpressionLinearInterpolate
   inputs [
    A {from=MaterialExpressionMultiply_1}
    B {from=MaterialExpressionReroute_1}
    Alpha {from=MaterialExpressionSaturate_7}
   ]
   properties {desc="Fixes normals\r\non backfaces"}
  }
  MaterialExpressionLinearInterpolate_3 {
   class=MaterialExpressionLinearInterpolate
   inputs [
    A {from=MaterialExpressionScalarParameter_18}
    B {from=MaterialExpressionScalarParameter_19}
    Alpha {from=MaterialExpressionPower_3}
   ]
  }
  MaterialExpressionLinearInterpolate_4 {
   class=MaterialExpressionLinearInterpolate
   inputs [
    A {from=MaterialExpressionTextureCoordinate_0}
    B {from=MaterialExpressionAdd_29}
    Alpha {from=MaterialExpressionLinearInterpolate_11}
   ]
  }
  MaterialExpressionLinearInterpolate_5 {
   class=MaterialExpressionLinearInterpolate
   inputs [
    A {from=MaterialExpressionScalarParameter_41}
    B {from=""}
    Alpha {from=MaterialExpressionScalarParameter_34}
   ]
   properties {constA=3}
  }
  MaterialExpressionLinearInterpolate_6 {
   class=MaterialExpressionLinearInterpolate
   inputs [
    A {from=MaterialExpressionMaterialFunctionCall_33, output=EmissiveMask}
    B {from=""}
    Alpha {from=MaterialExpressionMaterialFunctionCall_34, output=EmissiveMask}
   ]
  }
  MaterialExpressionLinearInterpolate_7 {
   class=MaterialExpressionLinearInterpolate
   inputs [
    A {from=MaterialExpressionMaterialFunctionCall_35, output=EmissiveMask}
    B {from=""}
    Alpha {from=MaterialExpressionMaterialFunctionCall_36, output=EmissiveMask}
   ]
  }
  MaterialExpressionLinearInterpolate_8 {
   class=MaterialExpressionLinearInterpolate
   inputs [
    A {from=MaterialExpressionLinearInterpolate_6}
    B {from=""}
    Alpha {from=MaterialExpressionLinearInterpolate_7}
   ]
  }
  MaterialExpressionLinearInterpolate_9 {
   class=MaterialExpressionLinearInterpolate
   inputs [
    A {from=""}
    B {from=""}
    Alpha {from=MaterialExpressionNamedRerouteUsage_13}
   ]
   properties {constA=1, constB=-0.5}
  }
  MaterialExpressionLocalPosition_0 {class=MaterialExpressionLocalPosition, inputs [], properties {includedOffsets=ExcludeOffsets}}
  MaterialExpressionLocalPosition_1 {class=MaterialExpressionLocalPosition, inputs [], properties {includedOffsets=ExcludeOffsets}}
  MaterialExpressionMaterialFunctionCall_0 {
   class=MaterialExpressionMaterialFunctionCall
   inputs [
    Alpha Threshold {from=MaterialExpressionLinearInterpolate_1}
    Offset {from=""}
    Random {from=""}
    Offset2 {from=""}
   ]
   properties {materialFunction=&DitherTemporalAA_Fortnite}
  }
  MaterialExpressionMaterialFunctionCall_1 {
   class=MaterialExpressionMaterialFunctionCall
   inputs [
    HitPosition {from=MaterialExpressionVectorParameter_0, output=RGB}
    HitTime {from=MaterialExpressionVectorParameter_0, output=RGB}
    HitDuration {from=MaterialExpressionScalarParameter_2}
    HitRadius {from=MaterialExpressionConstant_3}
    PixelPosition {from=MaterialExpressionWorldPosition_1, output=XYZ}
   ]
   properties {materialFunction=&/MF_SecretObject_Hit_Original}
  }
  MaterialExpressionMaterialFunctionCall_10 {
   class=MaterialExpressionMaterialFunctionCall
   inputs [
    In {from=MaterialExpressionMaterialFunctionCall_11, output=Result}
    UseSphericalMappingOverride {from=""}
   ]
   properties {materialFunction=/Game/Packages/Fortress_SharedMaterials/Base_Material_Functions/MF_CurieFallbackEffect}
  }
  MaterialExpressionMaterialFunctionCall_11 {
   class=MaterialExpressionMaterialFunctionCall
   inputs [Material Attributes {from=MaterialExpressionMaterialFunctionCall_8, output=Result}]
   properties {materialFunction=/Game/Creative/Sets/MilitaryBase/Props/MF_FanSpin}
  }
  MaterialExpressionMaterialFunctionCall_12 {
   class=MaterialExpressionMaterialFunctionCall
   inputs [MA In {from=MaterialExpressionReroute_10, output=Output}]
   properties {materialFunction=&MF_CreativeMoveTool}
  }
  MaterialExpressionMaterialFunctionCall_13 {
   class=MaterialExpressionMaterialFunctionCall
   inputs [
    Material_In {from=MaterialExpressionReroute_69, output=MaterialAttributes}
    UseDestructibleDamage {from=""}
    --- usage info tooltip --- {from=""}
    UseCrackedDamageTextures {from=""}
    UseDamageWobbleWPO {from=""}
    UseAdditiveWPO {from=""}
    WorldPosition {from=""}
    UseDestructionMultiplier {from=""}
    ExposeLerpFunctionForInvertingTheHitDistanceCalc {from=""}
    InvertDamageFalloff {from=""}
    UseAlternateHitGradientOutput {from=""}
    DestructibleDamageAmount {from=""}
    WSDeformationCenter {from=""}
    WSDamageRadius {from=""}
   ]
   properties {materialFunction=&MF_FN_DestructibleDamage}
  }
  MaterialExpressionMaterialFunctionCall_14 {
   class=MaterialExpressionMaterialFunctionCall
   inputs [
    In {from=MaterialExpressionMaterialFunctionCall_10, output=Result}
    Custom Mask {from=""}
    Custom Rotation Axis {from=""}
   ]
   properties {materialFunction=/Game/Athena/WinterArt/Materials/MF_WindWPO_v2}
  }
  MaterialExpressionMaterialFunctionCall_15 {
   class=MaterialExpressionMaterialFunctionCall
   inputs [In {from=MaterialExpressionMaterialFunctionCall_14, output=Result}]
   properties {materialFunction=/Game/Global/Base4Layers_2022/MaterialFunctions/MF_SnowAndIceArchitecture}
  }
  MaterialExpressionMaterialFunctionCall_16 {
   class=MaterialExpressionMaterialFunctionCall
   inputs [In {from=MaterialExpressionMaterialFunctionCall_15, output=Result}]
   properties {materialFunction=/Game/Global/Base4Layers_2022/MaterialFunctions/MF_VibrationRattle}
  }
  MaterialExpressionMaterialFunctionCall_17 {
   class=MaterialExpressionMaterialFunctionCall
   inputs [
    In {from=MaterialExpressionReroute_13, output=Result}
    "RVT Height Map (Roughness)" {from=MaterialExpressionGetMaterialAttributes_3, output=Roughness}
   ]
   properties {materialFunction=/Game/Global/Base4Layers_2022/MaterialFunctions/MF_RVTBlending}
  }
  MaterialExpressionMaterialFunctionCall_18 {
   class=MaterialExpressionMaterialFunctionCall
   inputs [
    AO {from=""}
    RVT_Height {from=MaterialExpressionReroute_12, output="RVT_Height (S)"}
    For Buildings {from=MaterialExpressionReroute_11, output="For Moss (MA)"}
    For Moss {from=""}
   ]
   properties {materialFunction=/Game/Environments/Asteria/Rocks/Materials/MaterialFunctions/MF_JungleBiome}
  }
  MaterialExpressionMaterialFunctionCall_19 {
   class=MaterialExpressionMaterialFunctionCall
   inputs []
   properties {materialFunction=&ObjectScale}
  }
  MaterialExpressionMaterialFunctionCall_2 {
   class=MaterialExpressionMaterialFunctionCall
   inputs [In {from=MaterialExpressionReroute_2, output=Output}]
   properties {materialFunction=/Game/ArtTools/Tech_Art/DirectFX/Assets/MaterialFunctions/MF_DirectFX}
  }
  MaterialExpressionMaterialFunctionCall_20 {
   class=MaterialExpressionMaterialFunctionCall
   inputs [
    Seed {from=""}
    Float {from=MaterialExpressionAdd_36}
    Vec3 {from=""}
    Vec2 {from=""}
   ]
   properties {materialFunction=&MF_Hash}
  }
  MaterialExpressionMaterialFunctionCall_21 {
   class=MaterialExpressionMaterialFunctionCall
   inputs [
    TemporalSampleIndex {from=""}
    PixelPosition {from=""}
   ]
   properties {materialFunction=/BaseMaterial/Materials/Functions/MF_Noise_InterleavedGradientGolden_1d}
  }
  MaterialExpressionMaterialFunctionCall_22 {
   class=MaterialExpressionMaterialFunctionCall
   inputs [
    InputBlend {from=MaterialExpressionMaterialFunctionCall_25, output=AxisMask}
    TriplanarDitherPattern {from=MaterialExpressionSubtract_1}
   ]
   properties {materialFunction=/BaseMaterial/Materials/Functions/MF_Blend_RGBToIndex}
  }
  MaterialExpressionMaterialFunctionCall_23 {
   class=MaterialExpressionMaterialFunctionCall
   inputs [
    Position {from=MaterialExpressionMultiply_40}
    Tiling {from=MaterialExpressionReroute_23}
    RefNorm {from=MaterialExpressionReroute_17}
    "Rotation (Turns)" {from=""}
    AxisIndex {from=MaterialExpressionReroute_18}
   ]
   properties {materialFunction=/BaseMaterial/Materials/Functions/MF_Coordinate_Triplanar}
  }
  MaterialExpressionMaterialFunctionCall_25 {
   class=MaterialExpressionMaterialFunctionCall
   inputs [
    RefNorm {from=MaterialExpressionNormalize_1}
    BlendFactor {from=MaterialExpressionScalarParameter_21}
   ]
   properties {materialFunction=/BaseMaterial/Materials/Functions/MF_Blend_TriplanarAxis}
  }
  MaterialExpressionMaterialFunctionCall_3 {
   class=MaterialExpressionMaterialFunctionCall
   inputs [
    HitPosition {from=MaterialExpressionVectorParameter_1, output=RGB}
    HitTime {from=MaterialExpressionVectorParameter_1, output=RGB}
    HitDuration {from=MaterialExpressionScalarParameter_2}
    HitRadius {from=MaterialExpressionConstant_3}
    PixelPosition {from=MaterialExpressionWorldPosition_1, output=XYZ}
   ]
   properties {materialFunction=&MF_SecretObject_Hit_Original}
  }
  MaterialExpressionMaterialFunctionCall_30 {
   class=MaterialExpressionMaterialFunctionCall
   inputs [
    UV1 {from=MaterialExpressionLinearInterpolate_4}
    UV2 {from=MaterialExpressionLinearInterpolate_10}
   ]
   properties {materialFunction=&/MF_GigaMix_2023_CustomUVs}
  }
  MaterialExpressionMaterialFunctionCall_31 {
   class=MaterialExpressionMaterialFunctionCall
   inputs [Emissive {from=MaterialExpressionAdd_54}]
   properties {materialFunction=&MF_FX_EmissiveAdjustments}
  }
  MaterialExpressionMaterialFunctionCall_32 {
   class=MaterialExpressionMaterialFunctionCall
   inputs [
    Alpha Threshold {from=MaterialExpressionMax_7}
    Offset {from=""}
    Random {from=""}
    Offset2 {from=""}
   ]
   properties {materialFunction=&DitherTemporalAA_Fortnite}
  }
  MaterialExpressionMaterialFunctionCall_33 {
   class=MaterialExpressionMaterialFunctionCall
   inputs [
    HitPosition {from=MaterialExpressionVectorParameter_12, output=RGB}
    HitDir {from=MaterialExpressionVectorParameter_17, output=RGBA}
    HitDuration {from=MaterialExpressionScalarParameter_52}
    HitRadius {from=MaterialExpressionScalarParameter_51}
    HitRadialNoiseSpeed {from=MaterialExpressionScalarParameter_53}
    HitRadialNoiseIntensity {from=MaterialExpressionScalarParameter_54}
    HitTime {from=MaterialExpressionVectorParameter_12, output=RGB}
    PixelPosition {from=MaterialExpressionNamedRerouteUsage_9}
   ]
   properties {materialFunction=&/MF_SecretObject_Hit}
  }
  MaterialExpressionMaterialFunctionCall_34 {
   class=MaterialExpressionMaterialFunctionCall
   inputs [
    HitPosition {from=MaterialExpressionVectorParameter_13, output=RGB}
    HitDir {from=MaterialExpressionVectorParameter_18, output=RGBA}
    HitDuration {from=MaterialExpressionScalarParameter_52}
    HitRadius {from=MaterialExpressionScalarParameter_51}
    HitRadialNoiseSpeed {from=MaterialExpressionScalarParameter_53}
    HitRadialNoiseIntensity {from=MaterialExpressionScalarParameter_54}
    HitTime {from=MaterialExpressionVectorParameter_13, output=RGB}
    PixelPosition {from=MaterialExpressionNamedRerouteUsage_9}
   ]
   properties {materialFunction=&MF_SecretObject_Hit}
  }
  MaterialExpressionMaterialFunctionCall_35 {
   class=MaterialExpressionMaterialFunctionCall
   inputs [
    HitPosition {from=MaterialExpressionVectorParameter_14, output=RGB}
    HitDir {from=MaterialExpressionVectorParameter_20, output=RGBA}
    HitDuration {from=MaterialExpressionScalarParameter_52}
    HitRadius {from=MaterialExpressionScalarParameter_51}
    HitRadialNoiseSpeed {from=MaterialExpressionScalarParameter_53}
    HitRadialNoiseIntensity {from=MaterialExpressionScalarParameter_54}
    HitTime {from=MaterialExpressionVectorParameter_14, output=RGB}
    PixelPosition {from=MaterialExpressionNamedRerouteUsage_9}
   ]
   properties {materialFunction=&MF_SecretObject_Hit}
  }
  MaterialExpressionMaterialFunctionCall_36 {
   class=MaterialExpressionMaterialFunctionCall
   inputs [
    HitPosition {from=MaterialExpressionVectorParameter_15, output=RGB}
    HitDir {from=MaterialExpressionVectorParameter_21, output=RGBA}
    HitDuration {from=MaterialExpressionScalarParameter_52}
    HitRadius {from=MaterialExpressionScalarParameter_51}
    HitRadialNoiseSpeed {from=MaterialExpressionScalarParameter_53}
    HitRadialNoiseIntensity {from=MaterialExpressionScalarParameter_54}
    HitTime {from=MaterialExpressionVectorParameter_15, output=RGB}
    PixelPosition {from=MaterialExpressionNamedRerouteUsage_9}
   ]
   properties {materialFunction=&MF_SecretObject_Hit}
  }
  MaterialExpressionMaterialFunctionCall_4 {
   class=MaterialExpressionMaterialFunctionCall
   inputs [
    HitPosition {from=MaterialExpressionVectorParameter_2, output=RGB}
    HitTime {from=MaterialExpressionVectorParameter_2, output=RGB}
    HitDuration {from=MaterialExpressionScalarParameter_2}
    HitRadius {from=MaterialExpressionConstant_3}
    PixelPosition {from=MaterialExpressionWorldPosition_1, output=XYZ}
   ]
   properties {materialFunction=&MF_SecretObject_Hit_Original}
  }
  MaterialExpressionMaterialFunctionCall_5 {
   class=MaterialExpressionMaterialFunctionCall
   inputs [
    HitPosition {from=MaterialExpressionVectorParameter_3, output=RGB}
    HitTime {from=MaterialExpressionVectorParameter_3, output=RGB}
    HitDuration {from=MaterialExpressionScalarParameter_2}
    HitRadius {from=MaterialExpressionConstant_3}
    PixelPosition {from=MaterialExpressionWorldPosition_1, output=XYZ}
   ]
   properties {materialFunction=&MF_SecretObject_Hit_Original}
  }
  MaterialExpressionMaterialFunctionCall_6 {
   class=MaterialExpressionMaterialFunctionCall
   inputs [In {from=MaterialExpressionMaterialFunctionCall_7, output=Result}]
   properties {materialFunction=/Game/ArtTools/Tech_Art/Veneer/MaterialFunctions/MF_Veneer}
  }
  MaterialExpressionMaterialFunctionCall_7 {
   class=MaterialExpressionMaterialFunctionCall
   inputs [In {from=MaterialExpressionMaterialFunctionCall_2, output=Result}]
   properties {materialFunction=/Game/Global/MaterialFunctions/MF_GlobalEmissiveMultiplier_MPC}
  }
  MaterialExpressionMaterialFunctionCall_8 {
   class=MaterialExpressionMaterialFunctionCall
   inputs [In {from=MaterialExpressionReroute_69, output=MaterialAttributes}]
   properties {materialFunction=/Game/Athena/WinterArt/Materials/MF_WPO_WindNew}
  }
  MaterialExpressionMaterialFunctionCall_9 {
   class=MaterialExpressionMaterialFunctionCall
   inputs [In {from=MaterialExpressionReroute_68, output=Result}]
   properties {materialFunction=/Game/Athena/WinterArt/Materials/MF_WPOAnim}
  }
  MaterialExpressionMaterialProxyReplace_0 {
   class=MaterialExpressionMaterialProxyReplace
   inputs [
    Realtime {from=MaterialExpressionLinearInterpolate_2, output=""}
    MaterialProxy {from=MaterialExpressionGetMaterialAttributes_0, output=Normal}
   ]
   properties {}
  }
  MaterialExpressionMaterialProxyReplace_1 {
   class=MaterialExpressionMaterialProxyReplace
   inputs [
    Realtime {from=MaterialExpressionMaterialFunctionCall_0, output=Result}
    MaterialProxy {from=MaterialExpressionConstant_5, output=""}
   ]
   properties {}
  }
  MaterialExpressionMax_0 {
   class=MaterialExpressionMax
   inputs [
    A {from=MaterialExpressionMaterialFunctionCall_1, output=Result}
    B {from=MaterialExpressionMax_1}
   ]
  }
  MaterialExpressionMax_1 {
   class=MaterialExpressionMax
   inputs [
    A {from=MaterialExpressionMaterialFunctionCall_3, output=Result}
    B {from=MaterialExpressionMax_2}
   ]
  }
  MaterialExpressionMax_10 {
   class=MaterialExpressionMax
   inputs [
    A {from=MaterialExpressionComponentMask_5}
    B {from=MaterialExpressionComponentMask_4}
   ]
  }
  MaterialExpressionMax_11 {
   class=MaterialExpressionMax
   inputs [
    A {from=MaterialExpressionMultiply_25}
    B {from=MaterialExpressionMultiply_44}
   ]
  }
  MaterialExpressionMax_12 {
   class=MaterialExpressionMax
   inputs [
    A {from=MaterialExpressionMax_17}
    B {from=MaterialExpressionSaturate_11}
   ]
  }
  MaterialExpressionMax_13 {
   class=MaterialExpressionMax
   inputs [
    A {from=MaterialExpressionMax_7}
    B {from=""}
   ]
   properties {constB=0}
  }
  MaterialExpressionMax_14 {
   class=MaterialExpressionMax
   inputs [
    A {from=MaterialExpressionSubtract_5}
    B {from=MaterialExpressionNamedRerouteUsage_28}
   ]
  }
  MaterialExpressionMax_15 {
   class=MaterialExpressionMax
   inputs [
    A {from=MaterialExpressionNamedRerouteUsage_10}
    B {from=MaterialExpressionNamedRerouteUsage_31}
   ]
  }
  MaterialExpressionMax_16 {
   class=MaterialExpressionMax
   inputs [
    A {from=MaterialExpressionMax_18}
    B {from=MaterialExpressionNamedRerouteUsage_23}
   ]
  }
  MaterialExpressionMax_17 {
   class=MaterialExpressionMax
   inputs [
    A {from=MaterialExpressionSaturate_12}
    B {from=MaterialExpressionNamedRerouteUsage_17}
   ]
  }
  MaterialExpressionMax_18 {
   class=MaterialExpressionMax
   inputs [
    A {from=MaterialExpressionNamedRerouteUsage_17}
    B {from=MaterialExpressionNamedRerouteUsage_20}
   ]
  }
  MaterialExpressionMax_19 {
   class=MaterialExpressionMax
   inputs [
    A {from=MaterialExpressionMultiply_67}
    B {from=MaterialExpressionSaturate_10}
   ]
  }
  MaterialExpressionMax_2 {
   class=MaterialExpressionMax
   inputs [
    A {from=MaterialExpressionMaterialFunctionCall_4, output=Result}
    B {from=MaterialExpressionMaterialFunctionCall_5, output=Result}
   ]
  }
  MaterialExpressionMax_20 {
   class=MaterialExpressionMax
   inputs [
    A {from=MaterialExpressionMax_14}
    B {from=MaterialExpressionNamedRerouteUsage_34}
   ]
  }
  MaterialExpressionMax_3 {
   class=MaterialExpressionMax
   inputs [
    A {from=MaterialExpressionMultiply_27}
    B {from=MaterialExpressionMultiply_28}
   ]
  }
  MaterialExpressionMax_4 {
   class=MaterialExpressionMax
   inputs [
    A {from=MaterialExpressionMultiply_36}
    B {from=MaterialExpressionMultiply_37}
   ]
  }
  MaterialExpressionMax_5 {
   class=MaterialExpressionMax
   inputs [
    A {from=MaterialExpressionMax_0}
    B {from=MaterialExpressionMultiply_5}
   ]
  }
  MaterialExpressionMax_6 {
   class=MaterialExpressionMax
   inputs [
    A {from=MaterialExpressionMax_5}
    B {from=MaterialExpressionNamedRerouteUsage_24}
   ]
  }
  MaterialExpressionMax_7 {
   class=MaterialExpressionMax
   inputs [
    A {from=MaterialExpressionIf_1}
    B {from=MaterialExpressionMultiply_46}
   ]
  }
  MaterialExpressionMax_8 {
   class=MaterialExpressionMax
   inputs [
    A {from=MaterialExpressionComponentMask_5}
    B {from=MaterialExpressionComponentMask_9}
   ]
  }
  MaterialExpressionMax_9 {
   class=MaterialExpressionMax
   inputs [
    A {from=MaterialExpressionComponentMask_9}
    B {from=MaterialExpressionComponentMask_4}
   ]
  }
  MaterialExpressionMin_0 {
   class=MaterialExpressionMin
   inputs [
    A {from=MaterialExpressionMaterialFunctionCall_33, output=Result}
    B {from=MaterialExpressionMaterialFunctionCall_34, output=Result}
   ]
   properties {}
  }
  MaterialExpressionMin_1 {
   class=MaterialExpressionMin
   inputs [
    A {from=MaterialExpressionMin_0, output=""}
    B {from=MaterialExpressionMin_2, output=""}
   ]
   properties {}
  }
  MaterialExpressionMin_2 {
   class=MaterialExpressionMin
   inputs [
    A {from=MaterialExpressionMaterialFunctionCall_35, output=Result}
    B {from=MaterialExpressionMaterialFunctionCall_36, output=Result}
   ]
   properties {}
  }
  MaterialExpressionMin_3 {
   class=MaterialExpressionMin
   inputs [
    A {from=MaterialExpressionCustom_0, output=""}
    B {from=MaterialExpressionCustom_1, output=""}
   ]
   properties {}
  }
  MaterialExpressionMin_4 {
   class=MaterialExpressionMin
   inputs [
    A {from=MaterialExpressionMin_3, output=""}
    B {from=MaterialExpressionCustom_2, output=""}
   ]
   properties {}
  }
  MaterialExpressionMultiply_0 {
   class=MaterialExpressionMultiply
   inputs [
    A {from=MaterialExpressionScalarParameter_1}
    B {from=""}
   ]
  }
  MaterialExpressionMultiply_1 {
   class=MaterialExpressionMultiply
   inputs [
    A {from=MaterialExpressionReroute_1}
    B {from=""}
   ]
   properties {constB=-1}
  }
  MaterialExpressionMultiply_10 {
   class=MaterialExpressionMultiply
   inputs [
    A {from=MaterialExpressionTextureSampleParameter2D_0, output=G}
    B {from=MaterialExpressionTextureSampleParameter2D_0, output=G}
   ]
  }
  MaterialExpressionMultiply_11 {
   class=MaterialExpressionMultiply
   inputs [
    A {from=MaterialExpressionComponentMask_1}
    B {from=MaterialExpressionScalarParameter_16}
   ]
  }
  MaterialExpressionMultiply_12 {
   class=MaterialExpressionMultiply
   inputs [
    A {from=MaterialExpressionTime_5}
    B {from=MaterialExpressionScalarParameter_17}
   ]
  }
  MaterialExpressionMultiply_13 {
   class=MaterialExpressionMultiply
   inputs [
    A {from=MaterialExpressionTime_7}
    B {from=MaterialExpressionVectorParameter_9, output=RGBA}
   ]
  }
  MaterialExpressionMultiply_14 {
   class=MaterialExpressionMultiply
   inputs [
    A {from=MaterialExpressionAdd_23}
    B {from=MaterialExpressionReroute_25}
   ]
   properties {constA=1}
  }
  MaterialExpressionMultiply_15 {
   class=MaterialExpressionMultiply
   inputs [
    A {from=MaterialExpressionComponentMask_24}
    B {from=MaterialExpressionSaturate_2}
   ]
  }
  MaterialExpressionMultiply_16 {
   class=MaterialExpressionMultiply
   inputs [
    A {from=MaterialExpressionComponentMask_23}
    B {from=MaterialExpressionSaturate_6}
   ]
  }
  MaterialExpressionMultiply_17 {
   class=MaterialExpressionMultiply
   inputs [
    A {from=MaterialExpressionComponentMask_22}
    B {from=MaterialExpressionSaturate_1}
   ]
  }
  MaterialExpressionMultiply_18 {
   class=MaterialExpressionMultiply
   inputs [
    A {from=MaterialExpressionVectorParameter_7, output=WiggleSpeed}
    B {from=MaterialExpressionTime_6}
   ]
  }
  MaterialExpressionMultiply_19 {
   class=MaterialExpressionMultiply
   inputs [
    A {from=MaterialExpressionDotProduct_0}
    B {from=MaterialExpressionNormalize_2}
   ]
  }
  MaterialExpressionMultiply_2 {
   class=MaterialExpressionMultiply
   inputs [
    A {from=MaterialExpressionTextureSample_0, output=R}
    B {from=MaterialExpressionScalarParameter_4}
   ]
   properties {constB=0.4}
  }
  MaterialExpressionMultiply_20 {
   class=MaterialExpressionMultiply
   inputs [
    A {from=MaterialExpressionTransform_1}
    B {from=MaterialExpressionVectorParameter_8, output=RGB}
   ]
  }
  MaterialExpressionMultiply_21 {
   class=MaterialExpressionMultiply
   inputs [
    A {from=MaterialExpressionTextureCoordinate_1}
    B {from=MaterialExpressionComponentMask_13}
   ]
   properties {constB=32}
  }
  MaterialExpressionMultiply_22 {
   class=MaterialExpressionMultiply
   inputs [
    A {from=MaterialExpressionReroute_22}
    B {from=MaterialExpressionSine_1}
   ]
  }
  MaterialExpressionMultiply_23 {
   class=MaterialExpressionMultiply
   inputs [
    A {from=MaterialExpressionAdd_19}
    B {from=MaterialExpressionVectorParameter_10, output=RGBA}
   ]
  }
  MaterialExpressionMultiply_24 {
   class=MaterialExpressionMultiply
   inputs [
    A {from=MaterialExpressionDivide_3}
    B {from=MaterialExpressionAdd_11}
   ]
  }
  MaterialExpressionMultiply_25 {
   class=MaterialExpressionMultiply
   inputs [
    A {from=MaterialExpressionOneMinus_6}
    B {from=MaterialExpressionPower_2}
   ]
  }
  MaterialExpressionMultiply_26 {
   class=MaterialExpressionMultiply
   inputs [
    A {from=MaterialExpressionTime_8}
    B {from=MaterialExpressionScalarParameter_27}
   ]
  }
  MaterialExpressionMultiply_27 {
   class=MaterialExpressionMultiply
   inputs [
    A {from=MaterialExpressionConvert_6}
    B {from=MaterialExpressionConvert_6}
   ]
  }
  MaterialExpressionMultiply_28 {
   class=MaterialExpressionMultiply
   inputs [
    A {from=MaterialExpressionConvert_6}
    B {from=MaterialExpressionConvert_6}
   ]
  }
  MaterialExpressionMultiply_29 {
   class=MaterialExpressionMultiply
   inputs [
    A {from=MaterialExpressionComponentMask_15}
    B {from=MaterialExpressionScalarParameter_15}
   ]
  }
  MaterialExpressionMultiply_3 {
   class=MaterialExpressionMultiply
   inputs [
    A {from=MaterialExpressionSubtract_2}
    B {from=MaterialExpressionScalarParameter_7}
   ]
   properties {constB=20}
  }
  MaterialExpressionMultiply_30 {
   class=MaterialExpressionMultiply
   inputs [
    A {from=MaterialExpressionTime_11}
    B {from=MaterialExpressionScalarParameter_20}
   ]
  }
  MaterialExpressionMultiply_31 {
   class=MaterialExpressionMultiply
   inputs [
    A {from=MaterialExpressionSubtract_9}
    B {from=MaterialExpressionScalarParameter_29}
   ]
  }
  MaterialExpressionMultiply_32 {
   class=MaterialExpressionMultiply
   inputs [
    A {from=MaterialExpressionAbs_4}
    B {from=MaterialExpressionAbs_4}
   ]
  }
  MaterialExpressionMultiply_33 {
   class=MaterialExpressionMultiply
   inputs [
    A {from=MaterialExpressionAbs_5}
    B {from=MaterialExpressionAbs_5}
   ]
  }
  MaterialExpressionMultiply_34 {
   class=MaterialExpressionMultiply
   inputs [
    A {from=MaterialExpressionTime_14}
    B {from=MaterialExpressionScalarParameter_24}
   ]
  }
  MaterialExpressionMultiply_35 {
   class=MaterialExpressionMultiply
   inputs [
    A {from=MaterialExpressionMultiply_51}
    B {from=MaterialExpressionSaturate_10}
   ]
  }
  MaterialExpressionMultiply_36 {
   class=MaterialExpressionMultiply
   inputs [
    A {from=MaterialExpressionConvert_7}
    B {from=MaterialExpressionConvert_7}
   ]
  }
  MaterialExpressionMultiply_37 {
   class=MaterialExpressionMultiply
   inputs [
    A {from=MaterialExpressionConvert_7}
    B {from=MaterialExpressionConvert_7}
   ]
  }
  MaterialExpressionMultiply_38 {
   class=MaterialExpressionMultiply
   inputs [
    A {from=MaterialExpressionNamedRerouteUsage_6}
    B {from=MaterialExpressionScalarParameter_33}
   ]
  }
  MaterialExpressionMultiply_39 {
   class=MaterialExpressionMultiply
   inputs [
    A {from=MaterialExpressionNamedRerouteUsage_0}
    B {from=MaterialExpressionLinearInterpolate_5}
   ]
   properties {constB=10}
  }
  MaterialExpressionMultiply_4 {
   class=MaterialExpressionMultiply
   inputs [
    A {from=MaterialExpressionTime_2}
    B {from=""}
   ]
   properties {constB=20}
  }
  MaterialExpressionMultiply_40 {
   class=MaterialExpressionMultiply
   inputs [
    A {from=MaterialExpressionTwoSidedSign_3}
    B {from=MaterialExpressionAdd_9}
   ]
  }
  MaterialExpressionMultiply_41 {
   class=MaterialExpressionMultiply
   inputs [
    A {from=MaterialExpressionSaturate_9}
    B {from=MaterialExpressionSaturate_9}
   ]
  }
  MaterialExpressionMultiply_42 {
   class=MaterialExpressionMultiply
   inputs [
    A {from=MaterialExpressionAdd_20}
    B {from=MaterialExpressionVectorParameter_11, output=RGB}
   ]
  }
  MaterialExpressionMultiply_43 {
   class=MaterialExpressionMultiply
   inputs [
    A {from=MaterialExpressionMultiply_10}
    B {from=MaterialExpressionSine_2}
   ]
  }
  MaterialExpressionMultiply_44 {
   class=MaterialExpressionMultiply
   inputs [
    A {from=MaterialExpressionAdd_31}
    B {from=MaterialExpressionSmoothStep_4}
   ]
  }
  MaterialExpressionMultiply_45 {
   class=MaterialExpressionMultiply
   inputs [
    A {from=MaterialExpressionTextureSampleParameter2D_0, output=R}
    B {from=MaterialExpressionTextureSampleParameter2D_0, output=R}
   ]
  }
  MaterialExpressionMultiply_46 {
   class=MaterialExpressionMultiply
   inputs [
    A {from=MaterialExpressionNamedRerouteUsage_8}
    B {from=MaterialExpressionMax_15}
   ]
   properties {constA=1}
  }
  MaterialExpressionMultiply_47 {
   class=MaterialExpressionMultiply
   inputs [
    A {from=MaterialExpressionAdd_37}
    B {from=MaterialExpressionConvert_3}
   ]
  }
  MaterialExpressionMultiply_48 {
   class=MaterialExpressionMultiply
   inputs [
    A {from=MaterialExpressionAdd_38}
    B {from=MaterialExpressionConvert_5}
   ]
  }
  MaterialExpressionMultiply_49 {
   class=MaterialExpressionMultiply
   inputs [
    A {from=MaterialExpressionAdd_39}
    B {from=MaterialExpressionConvert_9}
   ]
  }
  MaterialExpressionMultiply_5 {
   class=MaterialExpressionMultiply
   inputs [
    A {from=MaterialExpressionPower_0}
    B {from=MaterialExpressionScalarParameter_9}
   ]
   properties {constB=0.33}
  }
  MaterialExpressionMultiply_50 {
   class=MaterialExpressionMultiply
   inputs [
    A {from=MaterialExpressionTime_15}
    B {from=MaterialExpressionScalarParameter_42}
   ]
  }
  MaterialExpressionMultiply_51 {
   class=MaterialExpressionMultiply
   inputs [
    A {from=MaterialExpressionStep_0}
    B {from=MaterialExpressionStep_2}
   ]
  }
  MaterialExpressionMultiply_52 {
   class=MaterialExpressionMultiply
   inputs [
    A {from=MaterialExpressionTime_10}
    B {from=MaterialExpressionScalarParameter_46}
   ]
  }
  MaterialExpressionMultiply_53 {
   class=MaterialExpressionMultiply
   inputs [
    A {from=MaterialExpressionMax_3}
    B {from=""}
   ]
   properties {constB=2}
  }
  MaterialExpressionMultiply_54 {
   class=MaterialExpressionMultiply
   inputs [
    A {from=MaterialExpressionComponentMask_17}
    B {from=MaterialExpressionScalarParameter_47}
   ]
  }
  MaterialExpressionMultiply_55 {
   class=MaterialExpressionMultiply
   inputs [
    A {from=MaterialExpressionNamedRerouteUsage_25}
    B {from=MaterialExpressionNamedRerouteUsage_19}
   ]
  }
  MaterialExpressionMultiply_56 {
   class=MaterialExpressionMultiply
   inputs [
    A {from=MaterialExpressionMultiply_63}
    B {from=MaterialExpressionMultiply_61}
   ]
  }
  MaterialExpressionMultiply_57 {
   class=MaterialExpressionMultiply
   inputs [
    A {from=MaterialExpressionTime_12}
    B {from=MaterialExpressionScalarParameter_64}
   ]
  }
  MaterialExpressionMultiply_58 {
   class=MaterialExpressionMultiply
   inputs [
    A {from=MaterialExpressionFrac_9}
    B {from=MaterialExpressionComponentMask_19}
   ]
  }
  MaterialExpressionMultiply_59 {
   class=MaterialExpressionMultiply
   inputs [
    A {from=MaterialExpressionTime_13}
    B {from=MaterialExpressionScalarParameter_66}
   ]
  }
  MaterialExpressionMultiply_6 {
   class=MaterialExpressionMultiply
   inputs [
    A {from=MaterialExpressionAdd_2}
    B {from=MaterialExpressionScalarParameter_5}
   ]
   properties {constB=200}
  }
  MaterialExpressionMultiply_60 {
   class=MaterialExpressionMultiply
   inputs [
    A {from=MaterialExpressionTime_16}
    B {from=MaterialExpressionScalarParameter_48}
   ]
  }
  MaterialExpressionMultiply_61 {
   class=MaterialExpressionMultiply
   inputs [
    A {from=MaterialExpressionSmoothStep_9}
    B {from=MaterialExpressionSmoothStep_9}
   ]
  }
  MaterialExpressionMultiply_63 {
   class=MaterialExpressionMultiply
   inputs [
    A {from=MaterialExpressionFrac_12}
    B {from=MaterialExpressionScalarParameter_61}
   ]
  }
  MaterialExpressionMultiply_64 {
   class=MaterialExpressionMultiply
   inputs [
    A {from=MaterialExpressionSubtract_12}
    B {from=""}
   ]
   properties {constB=5}
  }
  MaterialExpressionMultiply_65 {
   class=MaterialExpressionMultiply
   inputs [
    A {from=MaterialExpressionMultiply_51}
    B {from=MaterialExpressionMax_19}
   ]
  }
  MaterialExpressionMultiply_66 {
   class=MaterialExpressionMultiply
   inputs [
    A {from=MaterialExpressionTime_19}
    B {from=MaterialExpressionScalarParameter_62}
   ]
  }
  MaterialExpressionMultiply_67 {
   class=MaterialExpressionMultiply
   inputs [
    A {from=MaterialExpressionIf_10}
    B {from=MaterialExpressionScalarParameter_67}
   ]
   properties {constB=2}
  }
  MaterialExpressionMultiply_68 {
   class=MaterialExpressionMultiply
   inputs [
    A {from=MaterialExpressionAdd_58}
    B {from=MaterialExpressionVectorParameter_19, output=RGB}
   ]
  }
  MaterialExpressionMultiply_7 {
   class=MaterialExpressionMultiply
   inputs [
    A {from=MaterialExpressionLocalPosition_0, output=XYZ}
    B {from=MaterialExpressionMaterialFunctionCall_19, output=Scale XYZ}
   ]
  }
  MaterialExpressionMultiply_8 {
   class=MaterialExpressionMultiply
   inputs [
    A {from=MaterialExpressionAdd_6}
    B {from=MaterialExpressionScalarParameter_11}
   ]
  }
  MaterialExpressionMultiply_9 {
   class=MaterialExpressionMultiply
   inputs [
    A {from=MaterialExpressionTime_4}
    B {from=MaterialExpressionScalarParameter_14}
   ]
  }
  MaterialExpressionNamedRerouteDeclaration_0 {
   class=MaterialExpressionNamedRerouteDeclaration
   inputs [None {from=MaterialExpressionMultiply_2}]
   properties {name=TriplanarNoise, nodeColor=<rgba 0 0.3438 1 1>}
  }
  MaterialExpressionNamedRerouteDeclaration_1 {
   class=MaterialExpressionNamedRerouteDeclaration
   inputs [None {from=MaterialExpressionAdd_53}]
   properties {name=GlitchyPos, nodeColor=<rgba 0 0.1094 1 1>}
  }
  MaterialExpressionNamedRerouteDeclaration_10 {
   class=MaterialExpressionNamedRerouteDeclaration
   inputs [None {from=MaterialExpressionSmoothStep_8}]
   properties {name=HitOpacity, nodeColor=<rgba 0 0.6484 1 1>}
  }
  MaterialExpressionNamedRerouteDeclaration_11 {
   class=MaterialExpressionNamedRerouteDeclaration
   inputs [None {from=MaterialExpressionMultiply_56}]
   properties {name=HitGlowEmissiveMask, nodeColor=<rgba 0.9688 1 0 1>}
  }
  MaterialExpressionNamedRerouteDeclaration_12 {
   class=MaterialExpressionNamedRerouteDeclaration
   inputs [None {from=MaterialExpressionLinearInterpolate_8}]
   properties {name=HitPulseMask, nodeColor=<rgba 0.7109 0 1 1>}
  }
  MaterialExpressionNamedRerouteDeclaration_13 {
   class=MaterialExpressionNamedRerouteDeclaration
   inputs [None {from=MaterialExpressionMultiply_33}]
   properties {name=BlockyWaveRaw, nodeColor=<rgba 0 1 0.3906 1>}
  }
  MaterialExpressionNamedRerouteDeclaration_14 {
   class=MaterialExpressionNamedRerouteDeclaration
   inputs [None {from=MaterialExpressionComponentMask_19}]
   properties {name=GlitchyGlitterSweep, nodeColor=<rgba 1 0.0703 0 1>}
  }
  MaterialExpressionNamedRerouteDeclaration_15 {
   class=MaterialExpressionNamedRerouteDeclaration
   inputs [None {from=MaterialExpressionVectorParameter_11, output=RGB}]
   properties {name=EmissiveColor, nodeColor=<rgba 0 1 0.4375 1>}
  }
  MaterialExpressionNamedRerouteDeclaration_16 {
   class=MaterialExpressionNamedRerouteDeclaration
   inputs [None {from=MaterialExpressionAdd_20}]
   properties {name=EmissiveAlpha, nodeColor=<rgba 1 0.1172 0 1>}
  }
  MaterialExpressionNamedRerouteDeclaration_17 {
   class=MaterialExpressionNamedRerouteDeclaration
   inputs [None {from=MaterialExpressionConstantBiasScale_3}]
   properties {name=GlitchBlocks, nodeColor=<rgba 1 0.8906 0 1>}
  }
  MaterialExpressionNamedRerouteDeclaration_18 {
   class=MaterialExpressionNamedRerouteDeclaration
   inputs [None {from=MaterialExpressionOneMinus_7}]
   properties {name=ProximityGitch, nodeColor=<rgba 0.7344 0 1 1>}
  }
  MaterialExpressionNamedRerouteDeclaration_2 {
   class=MaterialExpressionNamedRerouteDeclaration
   inputs [None {from=MaterialExpressionVectorNoise_0}]
   properties {name=Cellnoise, nodeColor=<rgba 0 0.4375 1 1>}
  }
  MaterialExpressionNamedRerouteDeclaration_3 {
   class=MaterialExpressionNamedRerouteDeclaration
   inputs [None {from=MaterialExpressionAdd_59}]
   properties {name=BlockyPos, nodeColor=<rgba 0.7578 1 0 1>}
  }
  MaterialExpressionNamedRerouteDeclaration_4 {
   class=MaterialExpressionNamedRerouteDeclaration
   inputs [None {from=MaterialExpressionMultiply_42}]
   properties {name=EmissiveTex, nodeColor=<rgba 1 0 0.4219 1>}
  }
  MaterialExpressionNamedRerouteDeclaration_5 {
   class=MaterialExpressionNamedRerouteDeclaration
   inputs [None {from=MaterialExpressionFrac_2}]
   properties {name=GlitchMask, nodeColor=<rgba 0.4062 0 1 1>}
  }
  MaterialExpressionNamedRerouteDeclaration_6 {
   class=MaterialExpressionNamedRerouteDeclaration
   inputs [None {from=MaterialExpressionSubtract_10}]
   properties {name=UVGlitchMask, nodeColor=<rgba 1 0.9375 0 1>}
  }
  MaterialExpressionNamedRerouteDeclaration_7 {
   class=MaterialExpressionNamedRerouteDeclaration
   inputs [None {from=MaterialExpressionLinearInterpolate_0}]
   properties {name=ProximityVis, nodeColor=<rgba 1 0 0.3281 1>}
  }
  MaterialExpressionNamedRerouteDeclaration_8 {
   class=MaterialExpressionNamedRerouteDeclaration
   inputs [None {from=MaterialExpressionSaturate_14}]
   properties {name=Sweep, nodeColor=<rgba 0.0078 1 0 1>}
  }
  MaterialExpressionNamedRerouteDeclaration_9 {
   class=MaterialExpressionNamedRerouteDeclaration
   inputs [None {from=MaterialExpressionMax_4}]
   properties {name=SweepBlocky, nodeColor=<rgba 1 0 0.3281 1>}
  }
  MaterialExpressionNamedRerouteUsage_0 {class=MaterialExpressionNamedRerouteUsage}
  MaterialExpressionNamedRerouteUsage_1 {class=MaterialExpressionNamedRerouteUsage}
  MaterialExpressionNamedRerouteUsage_10 {class=MaterialExpressionNamedRerouteUsage}
  MaterialExpressionNamedRerouteUsage_11 {class=MaterialExpressionNamedRerouteUsage}
  MaterialExpressionNamedRerouteUsage_12 {class=MaterialExpressionNamedRerouteUsage}
  MaterialExpressionNamedRerouteUsage_13 {class=MaterialExpressionNamedRerouteUsage}
  MaterialExpressionNamedRerouteUsage_14 {class=MaterialExpressionNamedRerouteUsage}
  MaterialExpressionNamedRerouteUsage_15 {class=MaterialExpressionNamedRerouteUsage}
  MaterialExpressionNamedRerouteUsage_16 {class=MaterialExpressionNamedRerouteUsage}
  MaterialExpressionNamedRerouteUsage_17 {class=MaterialExpressionNamedRerouteUsage}
  MaterialExpressionNamedRerouteUsage_18 {class=MaterialExpressionNamedRerouteUsage}
  MaterialExpressionNamedRerouteUsage_19 {class=MaterialExpressionNamedRerouteUsage}
  MaterialExpressionNamedRerouteUsage_2 {class=MaterialExpressionNamedRerouteUsage}
  MaterialExpressionNamedRerouteUsage_20 {class=MaterialExpressionNamedRerouteUsage}
  MaterialExpressionNamedRerouteUsage_21 {class=MaterialExpressionNamedRerouteUsage}
  MaterialExpressionNamedRerouteUsage_22 {class=MaterialExpressionNamedRerouteUsage}
  MaterialExpressionNamedRerouteUsage_23 {class=MaterialExpressionNamedRerouteUsage}
  MaterialExpressionNamedRerouteUsage_24 {class=MaterialExpressionNamedRerouteUsage}
  MaterialExpressionNamedRerouteUsage_25 {class=MaterialExpressionNamedRerouteUsage}
  MaterialExpressionNamedRerouteUsage_26 {class=MaterialExpressionNamedRerouteUsage}
  MaterialExpressionNamedRerouteUsage_27 {class=MaterialExpressionNamedRerouteUsage}
  MaterialExpressionNamedRerouteUsage_28 {class=MaterialExpressionNamedRerouteUsage}
  MaterialExpressionNamedRerouteUsage_29 {class=MaterialExpressionNamedRerouteUsage}
  MaterialExpressionNamedRerouteUsage_3 {class=MaterialExpressionNamedRerouteUsage}
  MaterialExpressionNamedRerouteUsage_30 {class=MaterialExpressionNamedRerouteUsage}
  MaterialExpressionNamedRerouteUsage_31 {class=MaterialExpressionNamedRerouteUsage}
  MaterialExpressionNamedRerouteUsage_32 {class=MaterialExpressionNamedRerouteUsage}
  MaterialExpressionNamedRerouteUsage_34 {class=MaterialExpressionNamedRerouteUsage}
  MaterialExpressionNamedRerouteUsage_4 {class=MaterialExpressionNamedRerouteUsage}
  MaterialExpressionNamedRerouteUsage_5 {class=MaterialExpressionNamedRerouteUsage}
  MaterialExpressionNamedRerouteUsage_6 {class=MaterialExpressionNamedRerouteUsage}
  MaterialExpressionNamedRerouteUsage_7 {class=MaterialExpressionNamedRerouteUsage}
  MaterialExpressionNamedRerouteUsage_8 {class=MaterialExpressionNamedRerouteUsage}
  MaterialExpressionNamedRerouteUsage_9 {class=MaterialExpressionNamedRerouteUsage}
  MaterialExpressionNoise_0 {
   class=MaterialExpressionNoise
   inputs [
    World Position {from=MaterialExpressionAdd_28, output=""}
    FilterWidth {from="", output=""}
   ]
   properties {bTurbulence=false, levels=3, noiseFunction=NOISEFUNCTION_GradientTex3D, scale=2.317}
  }
  MaterialExpressionNormalize_0 {class=MaterialExpressionNormalize, inputs [VectorInput {from=MaterialExpressionSubtract_0, output=""}]}
  MaterialExpressionNormalize_1 {class=MaterialExpressionNormalize, inputs [VectorInput {from=MaterialExpressionMultiply_20, output=""}]}
  MaterialExpressionNormalize_2 {class=MaterialExpressionNormalize, inputs [VectorInput {from=MaterialExpressionVectorParameter_16, output=RGB}]}
  MaterialExpressionObjectPositionWS_0 {class=MaterialExpressionObjectPositionWS, inputs [], properties {}}
  MaterialExpressionObjectPositionWS_1 {class=MaterialExpressionObjectPositionWS, inputs [], properties {}}
  MaterialExpressionObjectPositionWS_2 {class=MaterialExpressionObjectPositionWS, inputs [], properties {}}
  MaterialExpressionOneMinus_0 {class=MaterialExpressionOneMinus, inputs [None {from=MaterialExpressionMax_8}]}
  MaterialExpressionOneMinus_1 {class=MaterialExpressionOneMinus, inputs [None {from=MaterialExpressionDivide_7}]}
  MaterialExpressionOneMinus_10 {class=MaterialExpressionOneMinus, inputs [None {from=MaterialExpressionScalarParameter_63}]}
  MaterialExpressionOneMinus_2 {class=MaterialExpressionOneMinus, inputs [None {from=MaterialExpressionPower_1}]}
  MaterialExpressionOneMinus_3 {class=MaterialExpressionOneMinus, inputs [None {from=MaterialExpressionMax_6}]}
  MaterialExpressionOneMinus_4 {class=MaterialExpressionOneMinus, inputs [None {from=MaterialExpressionMax_9}]}
  MaterialExpressionOneMinus_5 {class=MaterialExpressionOneMinus, inputs [None {from=MaterialExpressionMax_10}]}
  MaterialExpressionOneMinus_6 {class=MaterialExpressionOneMinus, inputs [None {from=MaterialExpressionSaturate_9}]}
  MaterialExpressionOneMinus_7 {class=MaterialExpressionOneMinus, inputs [None {from=MaterialExpressionAbs_3}]}
  MaterialExpressionOneMinus_8 {class=MaterialExpressionOneMinus, inputs [None {from=MaterialExpressionMax_12}]}
  MaterialExpressionOneMinus_9 {class=MaterialExpressionOneMinus, inputs [None {from=MaterialExpressionNamedRerouteUsage_32}]}
  MaterialExpressionPinBase_0 {
   class=MaterialExpressionPinBase
   inputs []
   properties {reroutePins [Input {expression=&self:MaterialExpressionReroute_3}]}
  }
  MaterialExpressionPinBase_1 {
   class=MaterialExpressionPinBase
   inputs [Output {from=MaterialExpressionSetMaterialAttributes_3, output=""}]
   properties {reroutePins [Output {expression=&self:MaterialExpressionReroute_2}]}
  }
  MaterialExpressionPinBase_2 {
   class=MaterialExpressionPinBase
   inputs []
   properties {reroutePins [MaterialAttributes {expression=&self:MaterialExpressionReroute_70}]}
  }
  MaterialExpressionPinBase_3 {
   class=MaterialExpressionPinBase
   inputs [Output {from=MaterialExpressionSetMaterialAttributes_4, output=""}]
   properties {reroutePins [Output {expression=&self:MaterialExpressionReroute_71}]}
  }
  MaterialExpressionPinBase_4 {
   class=MaterialExpressionPinBase
   inputs []
   properties {reroutePins [MaterialAttributes {expression=&self:MaterialExpressionReroute_69}]}
  }
  MaterialExpressionPinBase_5 {
   class=MaterialExpressionPinBase
   inputs [Result {from=MaterialExpressionMaterialFunctionCall_16, output=Result}]
   properties {reroutePins [Result {expression=&self:MaterialExpressionReroute_68}]}
  }
  MaterialExpressionPinBase_6 {
   class=MaterialExpressionPinBase
   inputs []
   properties {reroutePins [InputPin {expression=&self:MaterialExpressionReroute_0}]}
  }
  MaterialExpressionPinBase_7 {
   class=MaterialExpressionPinBase
   inputs [Output {from=MaterialExpressionShadingPathSwitch_0, output=""}]
   properties {reroutePins [Output {expression=&self:MaterialExpressionReroute_10}]}
  }
  MaterialExpressionPinBase_8 {
   class=MaterialExpressionPinBase
   inputs []
   properties {reroutePins [
     "For Moss (MA)" {expression=&self:MaterialExpressionReroute_11}
     "RVT_Height (S)" {expression=&self:MaterialExpressionReroute_12}
    ]}
  }
  MaterialExpressionPinBase_9 {
   class=MaterialExpressionPinBase
   inputs [Result {from=MaterialExpressionMaterialFunctionCall_18, output=Buildings Jungl}]
   properties {reroutePins [Result {expression=&self:MaterialExpressionReroute_13}]}
  }
  MaterialExpressionPower_0 {
   class=MaterialExpressionPower
   inputs [
    Base {from=MaterialExpressionSaturate_3}
    Exp {from=MaterialExpressionScalarParameter_8}
   ]
  }
  MaterialExpressionPower_1 {
   class=MaterialExpressionPower
   inputs [
    Base {from=MaterialExpressionSaturate_4}
    Exp {from=""}
   ]
   properties {constExponent=0.5}
  }
  MaterialExpressionPower_2 {
   class=MaterialExpressionPower
   inputs [
    Base {from=MaterialExpressionSmoothStep_0}
    Exp {from=""}
   ]
  }
  MaterialExpressionPower_3 {
   class=MaterialExpressionPower
   inputs [
    Base {from=MaterialExpressionSaturate_9}
    Exp {from=""}
   ]
   properties {constExponent=0.25}
  }
  MaterialExpressionPower_4 {
   class=MaterialExpressionPower
   inputs [
    Base {from=MaterialExpressionMax_11}
    Exp {from=""}
   ]
  }
  MaterialExpressionQualitySwitch_0 {
   class=MaterialExpressionQualitySwitch
   inputs [
    Default {from=MaterialExpressionReroute_0, output=InputPin}
    Low {from=MaterialExpressionSetMaterialAttributes_5, output=""}
    High {from=MaterialExpressionReroute_0, output=InputPin}
    Medium {from=MaterialExpressionReroute_0, output=InputPin}
    Epic {from=MaterialExpressionReroute_0, output=InputPin}
   ]
   properties {}
  }
  MaterialExpressionReroute_0 {class=MaterialExpressionReroute, inputs [None {from=MaterialExpressionMaterialFunctionCall_17, output=Result}]}
  MaterialExpressionReroute_1 {class=MaterialExpressionReroute, inputs [None {from=MaterialExpressionTransform_0}]}
  MaterialExpressionReroute_10 {class=MaterialExpressionReroute, inputs [None {from=MaterialExpressionShadingPathSwitch_0}]}
  MaterialExpressionReroute_11 {class=MaterialExpressionReroute, inputs [None {from=MaterialExpressionMaterialFunctionCall_6, output=Result}]}
  MaterialExpressionReroute_12 {class=MaterialExpressionReroute, inputs [None {from=MaterialExpressionMaterialFunctionCall_17, output=RVT H}]}
  MaterialExpressionReroute_13 {class=MaterialExpressionReroute, inputs [None {from=MaterialExpressionMaterialFunctionCall_18, output=Buildings Jungl}]}
  MaterialExpressionReroute_16 {class=MaterialExpressionReroute, inputs [None {from=MaterialExpressionTransform_1}]}
  MaterialExpressionReroute_17 {class=MaterialExpressionReroute, inputs [None {from=MaterialExpressionReroute_16}]}
  MaterialExpressionReroute_18 {class=MaterialExpressionReroute, inputs [None {from=MaterialExpressionMaterialFunctionCall_22, output=AxisIndex}]}
  MaterialExpressionReroute_19 {class=MaterialExpressionReroute, inputs [None {from=MaterialExpressionTransform_2}]}
  MaterialExpressionReroute_2 {class=MaterialExpressionReroute, inputs [None {from=MaterialExpressionSetMaterialAttributes_3}]}
  MaterialExpressionReroute_20 {class=MaterialExpressionReroute, inputs [None {from=MaterialExpressionReroute_19}]}
  MaterialExpressionReroute_21 {class=MaterialExpressionReroute, inputs [None {from=MaterialExpressionVectorParameter_7, output=Tiling}]}
  MaterialExpressionReroute_22 {class=MaterialExpressionReroute, inputs [None {from=MaterialExpressionVectorParameter_7, output=WiggleIntensity}]}
  MaterialExpressionReroute_23 {class=MaterialExpressionReroute, inputs [None {from=MaterialExpressionDivide_5}]}
  MaterialExpressionReroute_25 {class=MaterialExpressionReroute, inputs [None {from=MaterialExpressionNamedRerouteUsage_6}]}
  MaterialExpressionReroute_26 {class=MaterialExpressionReroute, inputs [None {from=MaterialExpressionVectorParameter_6, output=RGB}]}
  MaterialExpressionReroute_27 {class=MaterialExpressionReroute, inputs [None {from=MaterialExpressionVectorParameter_6, output=A}]}
  MaterialExpressionReroute_3 {class=MaterialExpressionReroute, inputs [None {from=MaterialExpressionMaterialFunctionCall_9, output=Result}]}
  MaterialExpressionReroute_32 {class=MaterialExpressionReroute, inputs [None {from=MaterialExpressionMin_1}]}
  MaterialExpressionReroute_4 {class=MaterialExpressionReroute, inputs [None {from=MaterialExpressionComponentMask_4}]}
  MaterialExpressionReroute_5 {class=MaterialExpressionReroute, inputs [None {from=MaterialExpressionComponentMask_5}]}
  MaterialExpressionReroute_6 {class=MaterialExpressionReroute, inputs [None {from=MaterialExpressionReroute_5}]}
  MaterialExpressionReroute_68 {class=MaterialExpressionReroute, inputs [None {from=MaterialExpressionMaterialFunctionCall_16, output=Result}]}
  MaterialExpressionReroute_69 {class=MaterialExpressionReroute, inputs [None {from=MaterialExpressionReroute_71, output=Output}]}
  MaterialExpressionReroute_7 {class=MaterialExpressionReroute, inputs [None {from=MaterialExpressionReroute_4}]}
  MaterialExpressionReroute_70 {class=MaterialExpressionReroute, inputs [None {from=MaterialExpressionMaterialFunctionCall_30, output=Result}]}
  MaterialExpressionReroute_71 {class=MaterialExpressionReroute, inputs [None {from=MaterialExpressionSetMaterialAttributes_4}]}
  MaterialExpressionReroute_8 {class=MaterialExpressionReroute, inputs [None {from=MaterialExpressionComponentMask_9}]}
  MaterialExpressionReroute_9 {class=MaterialExpressionReroute, inputs [None {from=MaterialExpressionReroute_8}]}
  MaterialExpressionSaturate_0 {class=MaterialExpressionSaturate, inputs [None {from=MaterialExpressionTwoSidedSign_0}]}
  MaterialExpressionSaturate_1 {class=MaterialExpressionSaturate, inputs [None {from=MaterialExpressionFloor_1}]}
  MaterialExpressionSaturate_10 {class=MaterialExpressionSaturate, inputs [None {from=MaterialExpressionMax_20}]}
  MaterialExpressionSaturate_11 {class=MaterialExpressionSaturate, inputs [None {from=MaterialExpressionMultiply_39}]}
  MaterialExpressionSaturate_12 {class=MaterialExpressionSaturate, inputs [None {from=MaterialExpressionMultiply_64}]}
  MaterialExpressionSaturate_14 {class=MaterialExpressionSaturate, inputs [None {from=MaterialExpressionMultiply_53}]}
  MaterialExpressionSaturate_2 {class=MaterialExpressionSaturate, inputs [None {from=MaterialExpressionFloor_3}]}
  MaterialExpressionSaturate_3 {class=MaterialExpressionSaturate, inputs [None {from=MaterialExpressionMultiply_3}]}
  MaterialExpressionSaturate_4 {class=MaterialExpressionSaturate, inputs [None {from=MaterialExpressionDivide_9}]}
  MaterialExpressionSaturate_5 {class=MaterialExpressionSaturate, inputs [None {from=MaterialExpressionSign_0}]}
  MaterialExpressionSaturate_6 {class=MaterialExpressionSaturate, inputs [None {from=MaterialExpressionFloor_2}]}
  MaterialExpressionSaturate_7 {class=MaterialExpressionSaturate, inputs [None {from=MaterialExpressionTwoSidedSign_1}]}
  MaterialExpressionSaturate_8 {class=MaterialExpressionSaturate, inputs [None {from=MaterialExpressionTwoSidedSign_2}]}
  MaterialExpressionSaturate_9 {class=MaterialExpressionSaturate, inputs [None {from=MaterialExpressionDivide_13}]}
  MaterialExpressionScalarParameter_0 {class=MaterialExpressionScalarParameter, properties {defaultValue=1, parameterName=Visibility.InitialState}}
  MaterialExpressionScalarParameter_1 {
   class=MaterialExpressionScalarParameter
   properties {
    desc=Documentation link
    group="0 (Expand)  Documentation (RMB - Copy display name) and Debug view"
    parameterName=https://confluence.it.epicgames.com/display/FORT/Fortnite+Base+Material+2022
   }
  }
  MaterialExpressionScalarParameter_10 {class=MaterialExpressionScalarParameter, properties {defaultValue=0.5, parameterName=Visibility.Duration}}
  MaterialExpressionScalarParameter_11 {class=MaterialExpressionScalarParameter, properties {defaultValue=30, parameterName=BlockSize}}
  MaterialExpressionScalarParameter_12 {class=MaterialExpressionScalarParameter, properties {defaultValue=100, parameterName=HitRadius}}
  MaterialExpressionScalarParameter_13 {class=MaterialExpressionScalarParameter, properties {defaultValue=0.8, parameterName=GlitchThreshold}}
  MaterialExpressionScalarParameter_14 {class=MaterialExpressionScalarParameter, properties {defaultValue=2, parameterName=GlitchSpeed}}
  MaterialExpressionScalarParameter_15 {class=MaterialExpressionScalarParameter, properties {defaultValue=0.2}}
  MaterialExpressionScalarParameter_16 {class=MaterialExpressionScalarParameter, properties {defaultValue=0.5, parameterName=HitRadialNoiseIntensity}}
  MaterialExpressionScalarParameter_17 {class=MaterialExpressionScalarParameter, properties {defaultValue=0.1, parameterName=HitRadialNoiseSpeed}}
  MaterialExpressionScalarParameter_18 {class=MaterialExpressionScalarParameter, properties {defaultValue=2, parameterName=HitSplashGradientStart}}
  MaterialExpressionScalarParameter_19 {class=MaterialExpressionScalarParameter, properties {defaultValue=-2, parameterName=HitSplashGradientEnd}}
  MaterialExpressionScalarParameter_2 {class=MaterialExpressionScalarParameter, properties {defaultValue=0.5, parameterName=HitDuration}}
  MaterialExpressionScalarParameter_20 {class=MaterialExpressionScalarParameter, properties {defaultValue=0.15, parameterName=Param_1}}
  MaterialExpressionScalarParameter_21 {class=MaterialExpressionScalarParameter, properties {defaultValue=2, parameterName=TriplanarBlendFactor}}
  MaterialExpressionScalarParameter_22 {class=MaterialExpressionScalarParameter, properties {defaultValue=21300, parameterName=SweepNoiseSizeA}}
  MaterialExpressionScalarParameter_23 {class=MaterialExpressionScalarParameter, properties {defaultValue=-5100, parameterName=SweepNoiseSizeB}}
  MaterialExpressionScalarParameter_24 {class=MaterialExpressionScalarParameter, properties {defaultValue=16, parameterName=Param_2}}
  MaterialExpressionScalarParameter_25 {class=MaterialExpressionScalarParameter, properties {defaultValue=0.75, parameterName=Param_3}}
  MaterialExpressionScalarParameter_26 {class=MaterialExpressionScalarParameter, properties {defaultValue=0.05, parameterName=Param_4}}
  MaterialExpressionScalarParameter_27 {class=MaterialExpressionScalarParameter, properties {defaultValue=-0.237, parameterName=SweepSpeed}}
  MaterialExpressionScalarParameter_28 {class=MaterialExpressionScalarParameter, properties {defaultValue=2426.7727, parameterName=SweepSpacing}}
  MaterialExpressionScalarParameter_29 {class=MaterialExpressionScalarParameter, properties {defaultValue=0.5, parameterName=SweepNoiseContributionB}}
  MaterialExpressionScalarParameter_3 {class=MaterialExpressionScalarParameter, properties {parameterName=Visibility.Time}}
  MaterialExpressionScalarParameter_30 {class=MaterialExpressionScalarParameter, properties {defaultValue=0.9, parameterName=SweepWaveSharpnessA}}
  MaterialExpressionScalarParameter_31 {class=MaterialExpressionScalarParameter, properties {defaultValue=0.648, parameterName=SweepWaveSharpnessB}}
  MaterialExpressionScalarParameter_32 {class=MaterialExpressionScalarParameter, properties {defaultValue=1, parameterName=Visibility.InitialState}}
  MaterialExpressionScalarParameter_33 {class=MaterialExpressionScalarParameter, properties {defaultValue=20, parameterName=InteriorGlowIntensity}}
  MaterialExpressionScalarParameter_34 {class=MaterialExpressionScalarParameter, properties {defaultValue=1, parameterName=Visibility.InitialState}}
  MaterialExpressionScalarParameter_35 {class=MaterialExpressionScalarParameter, properties {defaultValue=0.35, parameterName=HitHoleInnerExterior}}
  MaterialExpressionScalarParameter_36 {class=MaterialExpressionScalarParameter, properties {defaultValue=0.54, parameterName=HitHoleInnerInterior}}
  MaterialExpressionScalarParameter_37 {class=MaterialExpressionScalarParameter, properties {defaultValue=1, parameterName=Visibility.InitialState}}
  MaterialExpressionScalarParameter_38 {class=MaterialExpressionScalarParameter, properties {defaultValue=0.5, parameterName=HitHoleOuter}}
  MaterialExpressionScalarParameter_39 {class=MaterialExpressionScalarParameter, properties {defaultValue=0.15, parameterName=HitGlowEdge}}
  MaterialExpressionScalarParameter_4 {class=MaterialExpressionScalarParameter, properties {defaultValue=0.4, parameterName=Hint.Time.PhaseDistortion, sliderMax=1}}
  MaterialExpressionScalarParameter_40 {class=MaterialExpressionScalarParameter, properties {defaultValue=0.5, parameterName=HitDuration}}
  MaterialExpressionScalarParameter_41 {class=MaterialExpressionScalarParameter, properties {defaultValue=10, parameterName=Param_5}}
  MaterialExpressionScalarParameter_42 {class=MaterialExpressionScalarParameter, properties {defaultValue=16, parameterName=Param_2}}
  MaterialExpressionScalarParameter_43 {class=MaterialExpressionScalarParameter, properties {defaultValue=0.15, parameterName=Param_6}}
  MaterialExpressionScalarParameter_44 {class=MaterialExpressionScalarParameter, properties {defaultValue=0.2, parameterName=Param_7}}
  MaterialExpressionScalarParameter_45 {class=MaterialExpressionScalarParameter, properties {defaultValue=-0.05, parameterName=SweepWaveBlockyTimeOffset}}
  MaterialExpressionScalarParameter_46 {class=MaterialExpressionScalarParameter, properties {defaultValue=2, parameterName=Param_8}}
  MaterialExpressionScalarParameter_47 {class=MaterialExpressionScalarParameter, properties {defaultValue=8, parameterName=Param_9}}
  MaterialExpressionScalarParameter_48 {class=MaterialExpressionScalarParameter, properties {defaultValue=5, parameterName=Param_10}}
  MaterialExpressionScalarParameter_49 {class=MaterialExpressionScalarParameter, properties {defaultValue=1, parameterName=Visibility.InitialState}}
  MaterialExpressionScalarParameter_5 {class=MaterialExpressionScalarParameter, properties {defaultValue=200, desc=In cm/s, parameterName=Hint.Time.OffsetSpeed}}
  MaterialExpressionScalarParameter_50 {class=MaterialExpressionScalarParameter, properties {defaultValue=0.4, parameterName=HitHoleOuterHidden}}
  MaterialExpressionScalarParameter_51 {class=MaterialExpressionScalarParameter, properties {defaultValue=100, parameterName=HitRadius}}
  MaterialExpressionScalarParameter_52 {class=MaterialExpressionScalarParameter, properties {defaultValue=0.5, parameterName=HitDuration}}
  MaterialExpressionScalarParameter_53 {class=MaterialExpressionScalarParameter, properties {defaultValue=0.1, parameterName=HitRadialNoiseSpeed}}
  MaterialExpressionScalarParameter_54 {class=MaterialExpressionScalarParameter, properties {defaultValue=0.5, parameterName=HitRadialNoiseIntensity}}
  MaterialExpressionScalarParameter_55 {class=MaterialExpressionScalarParameter, properties {defaultValue=0.15, parameterName=HitHoleInnerSecret}}
  MaterialExpressionScalarParameter_56 {class=MaterialExpressionScalarParameter, properties {defaultValue=0.25, parameterName=HitHoleInnerHidden}}
  MaterialExpressionScalarParameter_57 {class=MaterialExpressionScalarParameter, properties {defaultValue=0.25, parameterName=HitHoleOuterSecret}}
  MaterialExpressionScalarParameter_58 {class=MaterialExpressionScalarParameter, properties {defaultValue=0.15, parameterName=HitGlowEdgeSecret}}
  MaterialExpressionScalarParameter_59 {class=MaterialExpressionScalarParameter, properties {defaultValue=0.15, parameterName=HitGlowEdgeHidden}}
  MaterialExpressionScalarParameter_6 {class=MaterialExpressionScalarParameter, properties {defaultValue=800, desc=In cm, parameterName=Hint.Time.WaveLength}}
  MaterialExpressionScalarParameter_60 {class=MaterialExpressionScalarParameter, properties {defaultValue=1, parameterName=Visibility.InitialState}}
  MaterialExpressionScalarParameter_61 {class=MaterialExpressionScalarParameter, properties {defaultValue=1, parameterName=HitGlowIntensity}}
  MaterialExpressionScalarParameter_62 {class=MaterialExpressionScalarParameter, properties {defaultValue=0.1, parameterName=Param_11}}
  MaterialExpressionScalarParameter_63 {class=MaterialExpressionScalarParameter, properties {defaultValue=0.8, parameterName=GlitchThreshold}}
  MaterialExpressionScalarParameter_64 {class=MaterialExpressionScalarParameter, properties {defaultValue=0.1, parameterName=GlitchFlickerSpeed}}
  MaterialExpressionScalarParameter_65 {class=MaterialExpressionScalarParameter, properties {parameterName=GlitchFlickerThreshold}}
  MaterialExpressionScalarParameter_66 {class=MaterialExpressionScalarParameter, properties {defaultValue=2.5, parameterName=ProximityFlickerSpeed}}
  MaterialExpressionScalarParameter_67 {class=MaterialExpressionScalarParameter, properties {defaultValue=0.25, parameterName=Param_12}}
  MaterialExpressionScalarParameter_68 {class=MaterialExpressionScalarParameter, properties {defaultValue=100, parameterName=Param_13}}
  MaterialExpressionScalarParameter_69 {class=MaterialExpressionScalarParameter, properties {defaultValue=0.75, parameterName=SweepWaveGlitchSharpnessA}}
  MaterialExpressionScalarParameter_7 {class=MaterialExpressionScalarParameter, properties {defaultValue=40, desc=In cm, parameterName=Hint.Width}}
  MaterialExpressionScalarParameter_70 {class=MaterialExpressionScalarParameter, properties {defaultValue=0.6, parameterName=SweepWaveGlitchSharpnessB}}
  MaterialExpressionScalarParameter_8 {class=MaterialExpressionScalarParameter, properties {defaultValue=2, desc=Exponent, parameterName=Hint.Falloff}}
  MaterialExpressionScalarParameter_9 {class=MaterialExpressionScalarParameter, properties {defaultValue=0.333, parameterName=Hint.Intensity, sliderMax=1}}
  MaterialExpressionSetMaterialAttributes_0 {
   class=MaterialExpressionSetMaterialAttributes
   inputs [
    MaterialAttributes {from=MaterialExpressionBlendMaterialAttributes_0, output=""}
    Opacity Mask {from=MaterialExpressionMaterialProxyReplace_1, output=""}
   ]
   properties {attributeSetTypes [#3]}
  }
  MaterialExpressionSetMaterialAttributes_1 {
   class=MaterialExpressionSetMaterialAttributes
   inputs [
    MaterialAttributes {from=MaterialExpressionMaterialFunctionCall_12, output=MA Out}
    Emissive Color {from=MaterialExpressionMaterialFunctionCall_31, output=Result}
    Opacity Mask {from=MaterialExpressionMaterialFunctionCall_32, output=Result}
   ]
   properties {attributeSetTypes [#4, #3]}
  }
  MaterialExpressionSetMaterialAttributes_2 {
   class=MaterialExpressionSetMaterialAttributes
   inputs [
    MaterialAttributes {from="", output=""}
    Base Color {from=MaterialExpressionConstant_0, output=""}
    Metallic {from="", output=""}
    Specular {from=MaterialExpressionConstant_0, output=""}
    Roughness {from=MaterialExpressionConstant_1, output=""}
    Normal {from="", output=""}
    Emissive Color {from=MaterialExpressionConstant3Vector_2, output=""}
   ]
   properties {attributeSetTypes [
     #5
     #6
     #7
     #2
     #1
     #4
    ]}
  }
  MaterialExpressionSetMaterialAttributes_3 {
   class=MaterialExpressionSetMaterialAttributes
   inputs [
    MaterialAttributes {from=MaterialExpressionGetMaterialAttributes_0, output=MaterialAttributes}
    Normal {from=MaterialExpressionStaticSwitchParameter_0, output=""}
   ]
   properties {attributeSetTypes [#1]}
  }
  MaterialExpressionSetMaterialAttributes_4 {
   class=MaterialExpressionSetMaterialAttributes
   inputs [
    MaterialAttributes {from=MaterialExpressionReroute_70, output=MaterialAttributes}
    World Position Offset {from=MaterialExpressionMultiply_0, output=""}
   ]
   properties {attributeSetTypes [#8]}
  }
  MaterialExpressionSetMaterialAttributes_5 {
   class=MaterialExpressionSetMaterialAttributes
   inputs [
    MaterialAttributes {from=MaterialExpressionGetMaterialAttributes_1, output=MaterialAttributes}
    Specular {from=MaterialExpressionConstant_4, output=""}
    Roughness {from=MaterialExpressionConstant_2, output=""}
   ]
   properties {attributeSetTypes [#7, #2]}
  }
  MaterialExpressionShadingPathSwitch_0 {
   class=MaterialExpressionShadingPathSwitch
   inputs [
    Default {from=MaterialExpressionReroute_0, output=InputPin}
    Deferred {from="", output=""}
    Forward {from="", output=""}
    Mobile {from=MaterialExpressionQualitySwitch_0, output=""}
   ]
   properties {}
  }
  MaterialExpressionSign_0 {class=MaterialExpressionSign, inputs [None {from=MaterialExpressionScalarParameter_3, output=""}], properties {}}
  MaterialExpressionSine_0 {class=MaterialExpressionSine, inputs [None {from=MaterialExpressionDivide_8}]}
  MaterialExpressionSine_1 {class=MaterialExpressionSine, inputs [None {from=MaterialExpressionAdd_10}]}
  MaterialExpressionSine_2 {class=MaterialExpressionSine, inputs [None {from=MaterialExpressionAdd_7}]}
  MaterialExpressionSine_3 {class=MaterialExpressionSine, inputs [None {from=MaterialExpressionMultiply_66}]}
  MaterialExpressionSmoothStep_0 {
   class=MaterialExpressionSmoothStep
   inputs [
    Min {from="", output=""}
    Max {from="", output=""}
    Value {from=MaterialExpressionAbs_2, output=""}
   ]
   properties {constMax=0, constMin=1}
  }
  MaterialExpressionSmoothStep_1 {
   class=MaterialExpressionSmoothStep
   inputs [
    Min {from=MaterialExpressionConvert_8, output=""}
    Max {from="", output=""}
    Value {from=MaterialExpressionMultiply_32, output=""}
   ]
   properties {constMin=0.85}
  }
  MaterialExpressionSmoothStep_2 {
   class=MaterialExpressionSmoothStep
   inputs [
    Min {from=MaterialExpressionConvert_10, output=""}
    Max {from="", output=""}
    Value {from=MaterialExpressionMultiply_33, output=""}
   ]
   properties {constMin=0.85}
  }
  MaterialExpressionSmoothStep_3 {
   class=MaterialExpressionSmoothStep
   inputs [
    Min {from=MaterialExpressionScalarParameter_38, output=""}
    Max {from=MaterialExpressionIf_3, output=""}
    Value {from=MaterialExpressionAdd_27, output=""}
   ]
   properties {constMax=0.55, constMin=0.55}
  }
  MaterialExpressionSmoothStep_4 {
   class=MaterialExpressionSmoothStep
   inputs [
    Min {from=MaterialExpressionAdd_34, output=""}
    Max {from=MaterialExpressionScalarParameter_38, output=""}
    Value {from=MaterialExpressionAdd_27, output=""}
   ]
   properties {constMax=0.55, constMin=0.75}
  }
  MaterialExpressionSmoothStep_8 {
   class=MaterialExpressionSmoothStep
   inputs [
    Min {from=MaterialExpressionIf_5, output=""}
    Max {from=MaterialExpressionIf_6, output=""}
    Value {from=MaterialExpressionReroute_32, output=""}
   ]
   properties {constMax=0.55, constMin=0.55}
  }
  MaterialExpressionSmoothStep_9 {
   class=MaterialExpressionSmoothStep
   inputs [
    Min {from=MaterialExpressionIf_7, output=""}
    Max {from="", output=""}
    Value {from=MaterialExpressionAbs_6, output=""}
   ]
   properties {constMax=0, constMin=0.75}
  }
  MaterialExpressionSquareRoot_0 {class=MaterialExpressionSquareRoot, inputs [None {from=MaterialExpressionDotProduct_1, output=""}], properties {}}
  MaterialExpressionStaticSwitchParameter_0 {
   class=MaterialExpressionStaticSwitchParameter
   inputs [
    True {from=MaterialExpressionMaterialProxyReplace_0, output=""}
    False {from=MaterialExpressionLinearInterpolate_2, output=""}
   ]
   properties {parameterName=Force TangentSpaceNormalsFor HLOD ProxyReplace}
  }
  MaterialExpressionStep_0 {
   class=MaterialExpressionStep
   inputs [
    Y {from=MaterialExpressionFrac_4, output=""}
    X {from=MaterialExpressionScalarParameter_26, output=""}
   ]
   properties {}
  }
  MaterialExpressionStep_1 {
   class=MaterialExpressionStep
   inputs [
    Y {from=MaterialExpressionConstantBiasScale_4, output=""}
    X {from=MaterialExpressionNamedRerouteUsage_13, output=""}
   ]
   properties {}
  }
  MaterialExpressionStep_2 {
   class=MaterialExpressionStep
   inputs [
    Y {from=MaterialExpressionFrac_8, output=""}
    X {from=MaterialExpressionScalarParameter_44, output=""}
   ]
   properties {constX=0.1}
  }
  MaterialExpressionSubtract_0 {
   class=MaterialExpressionSubtract
   inputs [
    A {from=MaterialExpressionNamedRerouteUsage_7}
    B {from=MaterialExpressionReroute_26}
   ]
  }
  MaterialExpressionSubtract_1 {
   class=MaterialExpressionSubtract
   inputs [
    A {from=""}
    B {from=""}
   ]
   properties {constB=0.5}
  }
  MaterialExpressionSubtract_10 {
   class=MaterialExpressionSubtract
   inputs [
    A {from=MaterialExpressionMultiply_35}
    B {from=MaterialExpressionScalarParameter_43}
   ]
  }
  MaterialExpressionSubtract_11 {
   class=MaterialExpressionSubtract
   inputs [
    A {from=MaterialExpressionTime_9}
    B {from=MaterialExpressionReroute_27}
   ]
  }
  MaterialExpressionSubtract_12 {
   class=MaterialExpressionSubtract
   inputs [
    A {from=MaterialExpressionConstantBiasScale_4}
    B {from=MaterialExpressionLinearInterpolate_9}
   ]
  }
  MaterialExpressionSubtract_13 {
   class=MaterialExpressionSubtract
   inputs [
    A {from=MaterialExpressionReroute_32}
    B {from=MaterialExpressionIf_9}
   ]
  }
  MaterialExpressionSubtract_15 {
   class=MaterialExpressionSubtract
   inputs [
    A {from=MaterialExpressionNamedRerouteUsage_12}
    B {from=MaterialExpressionScalarParameter_63}
   ]
  }
  MaterialExpressionSubtract_16 {
   class=MaterialExpressionSubtract
   inputs [
    A {from=MaterialExpressionWorldPosition_8, output=XYZ}
    B {from=MaterialExpressionAdd_35}
   ]
  }
  MaterialExpressionSubtract_17 {
   class=MaterialExpressionSubtract
   inputs [
    A {from=MaterialExpressionSubtract_16}
    B {from=MaterialExpressionMultiply_19}
   ]
  }
  MaterialExpressionSubtract_2 {
   class=MaterialExpressionSubtract
   inputs [
    A {from=MaterialExpressionSine_0}
    B {from=MaterialExpressionOneMinus_1}
   ]
   properties {constB=0.95}
  }
  MaterialExpressionSubtract_3 {
   class=MaterialExpressionSubtract
   inputs [
    A {from=MaterialExpressionWorldPosition_3, output=Z}
    B {from=MaterialExpressionMultiply_6}
   ]
  }
  MaterialExpressionSubtract_4 {
   class=MaterialExpressionSubtract
   inputs [
    A {from=MaterialExpressionTime_3}
    B {from=MaterialExpressionAbs_0}
   ]
  }
  MaterialExpressionSubtract_5 {
   class=MaterialExpressionSubtract
   inputs [
    A {from=MaterialExpressionMaterialFunctionCall_20, output=Float}
    B {from=MaterialExpressionScalarParameter_25}
   ]
  }
  MaterialExpressionSubtract_6 {
   class=MaterialExpressionSubtract
   inputs [
    A {from=MaterialExpressionDivide_12}
    B {from=MaterialExpressionFrac_6}
   ]
  }
  MaterialExpressionSubtract_7 {
   class=MaterialExpressionSubtract
   inputs [
    A {from=MaterialExpressionMax_4}
    B {from=MaterialExpressionComponentMask_14}
   ]
  }
  MaterialExpressionSubtract_8 {
   class=MaterialExpressionSubtract
   inputs [
    A {from=MaterialExpressionTextureSample_2, output=RGBA}
    B {from=""}
   ]
   properties {constB=0.5}
  }
  MaterialExpressionSubtract_9 {
   class=MaterialExpressionSubtract
   inputs [
    A {from=MaterialExpressionTextureSample_3, output=RGBA}
    B {from=""}
   ]
   properties {constB=0.5}
  }
  MaterialExpressionTextureCoordinate_0 {class=MaterialExpressionTextureCoordinate}
  MaterialExpressionTextureCoordinate_1 {class=MaterialExpressionTextureCoordinate}
  MaterialExpressionTextureCoordinate_2 {class=MaterialExpressionTextureCoordinate, properties {coordinateIndex=1}}
  MaterialExpressionTextureObject_0 {
   class=MaterialExpressionTextureObject
   inputs []
   properties {samplerType=SAMPLERTYPE_Masks, texture=/Game/Global/Textures/Default/Masks_and_Gradients/CircuitBoardSeries}
  }
  MaterialExpressionTextureObject_1 {class=MaterialExpressionTextureObject, inputs [], properties {texture=/Game/Effects/Fort_Effects/Textures/Plasma/T_Plasma_Aurora_Seamless}}
  MaterialExpressionTextureSampleParameter2D_0 {
   class=MaterialExpressionTextureSampleParameter2D
   inputs [
    UVs {from=MaterialExpressionMaterialFunctionCall_23, output=UV}
    "DDX(UVs)" {from=MaterialExpressionComponentMask_2, output=""}
    "DDY(UVs)" {from=MaterialExpressionComponentMask_3, output=""}
    Apply View MipBias {from="", output=""}
   ]
   properties {
    mipValueMode=TMVM_Derivative
    parameterName=EmissiveTex
    samplerType=SAMPLERTYPE_Masks
    texture=&CircuitBoardSeries
   }
  }
  MaterialExpressionTextureSample_0 {
   class=MaterialExpressionTextureSample
   inputs [
    UVs {from=MaterialExpressionAdd_18}
    Tex {from=""}
    Apply View MipBias {from=""}
   ]
   properties {samplerType=SAMPLERTYPE_LinearColor, texture=&HighContrastNoise}
  }
  MaterialExpressionTextureSample_2 {
   class=MaterialExpressionTextureSample
   inputs [
    UVs {from=MaterialExpressionAdd_12}
    Tex {from=""}
    Apply View MipBias {from=""}
   ]
   properties {samplerType=SAMPLERTYPE_LinearColor, texture=&HighContrastNoise}
  }
  MaterialExpressionTextureSample_3 {
   class=MaterialExpressionTextureSample
   inputs [
    UVs {from=MaterialExpressionAdd_13}
    Tex {from=""}
    Apply View MipBias {from=""}
   ]
   properties {samplerType=SAMPLERTYPE_LinearColor, texture=&HighContrastNoise}
  }
  MaterialExpressionTime_0 {class=MaterialExpressionTime}
  MaterialExpressionTime_1 {class=MaterialExpressionTime}
  MaterialExpressionTime_10 {class=MaterialExpressionTime}
  MaterialExpressionTime_11 {class=MaterialExpressionTime}
  MaterialExpressionTime_12 {class=MaterialExpressionTime}
  MaterialExpressionTime_13 {class=MaterialExpressionTime}
  MaterialExpressionTime_14 {class=MaterialExpressionTime}
  MaterialExpressionTime_15 {class=MaterialExpressionTime}
  MaterialExpressionTime_16 {class=MaterialExpressionTime}
  MaterialExpressionTime_17 {class=MaterialExpressionTime}
  MaterialExpressionTime_19 {class=MaterialExpressionTime}
  MaterialExpressionTime_2 {class=MaterialExpressionTime}
  MaterialExpressionTime_3 {class=MaterialExpressionTime}
  MaterialExpressionTime_4 {class=MaterialExpressionTime}
  MaterialExpressionTime_5 {class=MaterialExpressionTime}
  MaterialExpressionTime_6 {class=MaterialExpressionTime}
  MaterialExpressionTime_7 {class=MaterialExpressionTime}
  MaterialExpressionTime_8 {class=MaterialExpressionTime}
  MaterialExpressionTime_9 {class=MaterialExpressionTime}
  MaterialExpressionTransformPosition_0 {
   class=MaterialExpressionTransformPosition
   inputs [None {from=MaterialExpressionDivide_14, output=""}]
   properties {transformType=TRANSFORMPOSSOURCE_World}
  }
  MaterialExpressionTransformPosition_2 {
   class=MaterialExpressionTransformPosition
   inputs [None {from=MaterialExpressionDivide_17, output=""}]
   properties {transformType=TRANSFORMPOSSOURCE_World}
  }
  MaterialExpressionTransform_0 {class=MaterialExpressionTransform, inputs [None {from=MaterialExpressionGetMaterialAttributes_0, output=Normal}]}
  MaterialExpressionTransform_1 {
   class=MaterialExpressionTransform
   inputs [None {from=MaterialExpressionVertexNormalWS_0}]
   properties {transformSourceType=TRANSFORMSOURCE_Local}
  }
  MaterialExpressionTransform_2 {
   class=MaterialExpressionTransform
   inputs [None {from=MaterialExpressionWorldPosition_5, output=XYZ}]
   properties {transformSourceType=TRANSFORMSOURCE_Local}
  }
  MaterialExpressionTwoSidedSign_0 {class=MaterialExpressionTwoSidedSign, inputs [], properties {}}
  MaterialExpressionTwoSidedSign_1 {class=MaterialExpressionTwoSidedSign, inputs [], properties {}}
  MaterialExpressionTwoSidedSign_2 {class=MaterialExpressionTwoSidedSign, inputs [], properties {}}
  MaterialExpressionTwoSidedSign_3 {class=MaterialExpressionTwoSidedSign, inputs [], properties {}}
  MaterialExpressionTwoSidedSign_4 {class=MaterialExpressionTwoSidedSign, inputs [], properties {}}
  MaterialExpressionVectorNoise_0 {class=MaterialExpressionVectorNoise, inputs [World Position {from=MaterialExpressionAdd_4, output=""}], properties {}}
  MaterialExpressionVectorNoise_1 {class=MaterialExpressionVectorNoise, inputs [World Position {from=MaterialExpressionAppendVector_1, output=""}], properties {}}
  MaterialExpressionVectorParameter_0 {
   class=MaterialExpressionVectorParameter
   properties {group=Hits, parameterName=Hit0, channelNames {a=Time, b=Z, g=Y, r=X}, defaultValue=<rgba -150 0 300 0>}
  }
  MaterialExpressionVectorParameter_1 {
   class=MaterialExpressionVectorParameter
   properties {group=Hits, parameterName=Hit1, primitiveDataIndex=4, channelNames {a=Time, b=Z, g=Y, r=X}, defaultValue=<rgba 150 0 300 0>}
  }
  MaterialExpressionVectorParameter_10 {class=MaterialExpressionVectorParameter, properties {parameterName=SweepNoiseChannelContribution, defaultValue=<rgba 0.75 0.5 0.75 0.5>}}
  MaterialExpressionVectorParameter_11 {class=MaterialExpressionVectorParameter, properties {parameterName=Emissive, defaultValue=<rgba 0.2065 0.8421 1 1>}}
  MaterialExpressionVectorParameter_12 {
   class=MaterialExpressionVectorParameter
   properties {group=Hits, parameterName=Hit0, channelNames {a=Time, b=Z, g=Y, r=X}, defaultValue=<rgba -150 0 300 0>}
  }
  MaterialExpressionVectorParameter_13 {
   class=MaterialExpressionVectorParameter
   properties {group=Hits, parameterName=Hit1, primitiveDataIndex=4, channelNames {a=Time, b=Z, g=Y, r=X}, defaultValue=<rgba 150 0 300 0>}
  }
  MaterialExpressionVectorParameter_14 {
   class=MaterialExpressionVectorParameter
   properties {group=Hits, parameterName=Hit2, primitiveDataIndex=8, channelNames {a=Time, b=Z, g=Y, r=X}, defaultValue=<rgba -150 0 100 0>}
  }
  MaterialExpressionVectorParameter_15 {
   class=MaterialExpressionVectorParameter
   properties {group=Hits, parameterName=Hit3, primitiveDataIndex=12, channelNames {a=Time, b=Z, g=Y, r=X}, defaultValue=<rgba 150 0 100 0>}
  }
  MaterialExpressionVectorParameter_16 {class=MaterialExpressionVectorParameter, properties {parameterName=Param_22, defaultValue=<rgba 1 0 0 128>}}
  MaterialExpressionVectorParameter_17 {
   class=MaterialExpressionVectorParameter
   properties {group=Hits, parameterName=HitDir0, channelNames {a=UseDir, b=Z, g=Y, r=X}, defaultValue=<rgba 0 1 0 0>}
  }
  MaterialExpressionVectorParameter_18 {
   class=MaterialExpressionVectorParameter
   properties {group=Hits, parameterName=HitDir1, primitiveDataIndex=4, channelNames {a=UseDir, b=Z, g=Y, r=X}, defaultValue=<rgba 0 1 0 0>}
  }
  MaterialExpressionVectorParameter_19 {class=MaterialExpressionVectorParameter, properties {parameterName=BlockSize, defaultValue=<rgba 32 32 16 0>}}
  MaterialExpressionVectorParameter_2 {
   class=MaterialExpressionVectorParameter
   properties {group=Hits, parameterName=Hit2, primitiveDataIndex=8, channelNames {a=Time, b=Z, g=Y, r=X}, defaultValue=<rgba -150 0 100 0>}
  }
  MaterialExpressionVectorParameter_20 {
   class=MaterialExpressionVectorParameter
   properties {group=Hits, parameterName=HitDir2, primitiveDataIndex=8, channelNames {a=UseDir, b=Z, g=Y, r=X}, defaultValue=<rgba 0 1 0 0>}
  }
  MaterialExpressionVectorParameter_21 {
   class=MaterialExpressionVectorParameter
   properties {group=Hits, parameterName=HitDir3, primitiveDataIndex=12, channelNames {a=UseDir, b=Z, g=Y, r=X}, defaultValue=<rgba 0 1 0 0>}
  }
  MaterialExpressionVectorParameter_3 {
   class=MaterialExpressionVectorParameter
   properties {group=Hits, parameterName=Hit3, primitiveDataIndex=12, channelNames {a=Time, b=Z, g=Y, r=X}, defaultValue=<rgba 150 0 100 0>}
  }
  MaterialExpressionVectorParameter_4 {class=MaterialExpressionVectorParameter, properties {parameterName=Param_1, defaultValue=<rgba 8 32 0 0>}}
  MaterialExpressionVectorParameter_5 {class=MaterialExpressionVectorParameter, properties {parameterName=Param_2, defaultValue=<rgba 10 0 100 0>}}
  MaterialExpressionVectorParameter_6 {class=MaterialExpressionVectorParameter, properties {defaultValue=<rgba 10 0 200 0>}}
  MaterialExpressionVectorParameter_7 {
   class=MaterialExpressionVectorParameter
   properties {
    parameterName=EmissiveTilingAndWiggle
    channelNames {a=WiggleSpeed, b=WiggleIntensity, g=WiggleWidth, r=Tiling}
    defaultValue=<rgba 370.9059 184 10 0>
   }
  }
  MaterialExpressionVectorParameter_8 {class=MaterialExpressionVectorParameter, properties {parameterName=TriplanarNormalBias, defaultValue=<rgba 1 2 1 0>}}
  MaterialExpressionVectorParameter_9 {class=MaterialExpressionVectorParameter, properties {parameterName=Param_3, defaultValue=<rgba 0.0312 0.0125 0.0253 0.0625>}}
  MaterialExpressionVertexNormalWS_0 {class=MaterialExpressionVertexNormalWS, inputs [], properties {}}
  MaterialExpressionVertexNormalWS_4 {class=MaterialExpressionVertexNormalWS, inputs [], properties {}}
  MaterialExpressionWorldPosition_0 {class=MaterialExpressionWorldPosition, inputs [], properties {worldPositionShaderOffset=WPT_ExcludeAllShaderOffsets}}
  MaterialExpressionWorldPosition_1 {class=MaterialExpressionWorldPosition, inputs [], properties {worldPositionShaderOffset=WPT_ExcludeAllShaderOffsets}}
  MaterialExpressionWorldPosition_2 {class=MaterialExpressionWorldPosition, inputs [], properties {worldPositionShaderOffset=WPT_ExcludeAllShaderOffsets}}
  MaterialExpressionWorldPosition_3 {class=MaterialExpressionWorldPosition, inputs [], properties {worldPositionShaderOffset=WPT_ExcludeAllShaderOffsets}}
  MaterialExpressionWorldPosition_4 {class=MaterialExpressionWorldPosition, inputs [], properties {worldPositionShaderOffset=WPT_ExcludeAllShaderOffsets}}
  MaterialExpressionWorldPosition_5 {class=MaterialExpressionWorldPosition, inputs [], properties {worldPositionShaderOffset=WPT_ExcludeAllShaderOffsets}}
  MaterialExpressionWorldPosition_6 {class=MaterialExpressionWorldPosition, inputs [], properties {}}
  MaterialExpressionWorldPosition_7 {class=MaterialExpressionWorldPosition, inputs [], properties {worldPositionShaderOffset=WPT_ExcludeAllShaderOffsets}}
  MaterialExpressionWorldPosition_8 {class=MaterialExpressionWorldPosition, inputs [], properties {}}
 }
 properties {
  bAutomaticallySetUsageInEditor=false
  bEnableTessellation=true
  bHasPixelAnimation=true
  bTangentSpaceNormal=false
  bUseMaterialAttributes=true
  bUsedWithNanite=true
  blendMode=BLEND_Masked
  opacityMaskClipValue=0.5
  previewMesh=/Hera/Sets/SigmaSanctum/Meshes/SM_SigmaSanctum_MazeWall_A_A
  usedShadingModels=Default Lit
  displacementScaling {center=0.5, magnitude=15}
 }
}
